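# Control Rig for the F_MED_BaskIsle player body cosmetic that procedurally poses its skirt bones to follow leg motion and movement states.
# Spherical pose readers on both thighs weight additive-style bone transforms for front, side and back skirt bones, with center bones taking the stronger side.
# One function graph also shapes these weights through curves.
# Crouch, slide, sprint, skydive and rail-grind variables gate further corrections, some scaled by pelvis height or spine bend.
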ControlRigBlueprint /BRCosmetics/Characters/Player/Female/Medium/Bodies/F_MED_BaskIsle/Meshes/F_MED_BaskIsle_CtrlRig
graphs [
 New Function_ContainedGraph {nodes [
   RigVMFunctionEntryNode()
   > RigVMFunction_Sequence = RigVMAggregateNode()
   n1 = &RigUnit_SphericalPoseReader(
    ActiveRegionScaleFactors=(PositiveWidth=1,NegativeWidth=1,PositiveHeight=1,NegativeHeight=1)
    ActiveRegionSize=0.05
    Debug=(bDrawDebug=false,bDraw2D=False,bDrawLocalAxes=False,DebugScale=25,DebugSegments=20,DebugThickness=0.25)
    DriverAxis=(X=1,Y=0,Z=0)
    DriverItem=(Type=Bone,Name="thigh_r")
    FalloffRegionScaleFactors=(PositiveWidth=1,NegativeWidth=1,PositiveHeight=0.9,NegativeHeight=0.9)
    FalloffSize=0.65
    FlipHeightScaling=false
    FlipWidthScaling=false
    OptionalParentItem=(Type=Bone,Name="None")
    RotationOffset=(X=-90,Y=0,Z=-30)
    ExecutePin=@RigVMFunction_Sequence.A
   )
   > &RigUnit_ModifyTransforms(
    ItemToModify="((Item=(Type=Bone,Name=\"base_skirt_front_r\"),Transform=(Rotation=(X=-0.183,Y=-0.183,Z=-0.683,W=0.683),Translation=(X=10,Y=0,Z=0),Scale3D=(X=1,Y=1,Z=1))),(Item=(Type=Bone,Name=\"base_skirt_side_r\"),Transform=(Rotation=(X=0.1116,Y=-0.633,Z=0.133,W=0.7544),Tra ...(truncated)"
    WeightMinimum=0
    Weight=@n1.OutputParam
   )
   > n2 = &RigUnit_SphericalPoseReader(
    ActiveRegionScaleFactors=(PositiveWidth=1,NegativeWidth=1,PositiveHeight=1,NegativeHeight=1.3)
    ActiveRegionSize=0.05
    Debug=(bDrawDebug=false,bDraw2D=False,bDrawLocalAxes=False,DebugScale=25,DebugSegments=20,DebugThickness=0.25)
    DriverAxis=(X=1,Y=0,Z=0)
    DriverItem=(Type=Bone,Name="thigh_r")
    FalloffRegionScaleFactors=(PositiveWidth=1,NegativeWidth=1,PositiveHeight=1,NegativeHeight=1.3)
    FalloffSize=0.2
    FlipHeightScaling=true
    FlipWidthScaling=true
    OptionalParentItem=(Type=Bone,Name="None")
    RotationOffset=(X=-40,Y=30,Z=0)
   )
   > &RigUnit_ModifyTransforms(
    ItemToModify="((Item=(Type=Bone,Name=\"base_skirt_front_r\"),Transform=(Rotation=(X=0.0249,Y=0.1281,Z=-0.1892,W=0.9732),Translation=(X=5,Y=0,Z=0),Scale3D=(X=1,Y=1,Z=1))),(Item=(Type=Bone,Name=\"base_skirt_side_r\"),Transform=(Rotation=(X=0,Y=0,Z=0,W=1),Translation=(X=0,Y=0, ...(truncated)"
    WeightMinimum=0
    Weight=@n2.OutputParam
   )
   > n3 = &RigUnit_SphericalPoseReader(
    ActiveRegionScaleFactors=(PositiveWidth=1,NegativeWidth=1,PositiveHeight=1,NegativeHeight=1)
    ActiveRegionSize=0.05
    Debug=(bDrawDebug=false,bDraw2D=False,bDrawLocalAxes=False,DebugScale=25,DebugSegments=20,DebugThickness=0.25)
    DriverAxis=(X=1,Y=0,Z=0)
    DriverItem=(Type=Bone,Name="thigh_r")
    FalloffRegionScaleFactors=(PositiveWidth=1,NegativeWidth=1,PositiveHeight=0.9,NegativeHeight=0.9)
    FalloffSize=0.65
    FlipHeightScaling=true
    FlipWidthScaling=true
    OptionalParentItem=(Type=Bone,Name="None")
    RotationOffset=(X=0,Y=27,Z=0)
   )
   > &RigUnit_ModifyTransforms(
    ItemToModify="((Item=(Type=Bone,Name=\"offset_skirt_side_r\"),Transform=(Rotation=(X=0,Y=0,Z=-0.766,W=0.6428),Translation=(X=5,Y=0,Z=0),Scale3D=(X=1,Y=1,Z=1))),(Item=(Type=Bone,Name=\"offset_skirt_front_r\"),Transform=(Rotation=(X=-0.0223,Y=0.532,Z=-0.2006,W=0.8224),Transla ...(truncated)"
    WeightMinimum=0
    Weight=@n3.OutputParam
   )
   > n4 = &RigUnit_SphericalPoseReader(
    ActiveRegionScaleFactors=(PositiveWidth=1,NegativeWidth=1,PositiveHeight=1,NegativeHeight=1)
    ActiveRegionSize=0.05
    Debug=(bDrawDebug=false,bDraw2D=False,bDrawLocalAxes=False,DebugScale=25,DebugSegments=20,DebugThickness=0.25)
    DriverAxis=(X=1,Y=0,Z=0)
    DriverItem=(Type=Bone,Name="thigh_r")
    FalloffRegionScaleFactors=(PositiveWidth=1,NegativeWidth=1,PositiveHeight=1,NegativeHeight=1)
    FalloffSize=0.55
    FlipHeightScaling=False
    FlipWidthScaling=False
    OptionalParentItem=(Type=Bone,Name="None")
    RotationOffset=(X=90,Y=0,Z=20)
   )
   > &RigUnit_ModifyTransforms(
    ItemToModify="((Item=(Type=Bone,Name=\"base_skirt_back_r\"),Transform=(Rotation=(X=0.1782,Y=0.1877,Z=-0.6649,W=0.7007),Translation=(X=2.5,Y=0,Z=0),Scale3D=(X=1,Y=1,Z=1))),(Item=(Type=Bone,Name=\"base_skirt_side_r\"),Transform=(Rotation=(X=-0.0334,Y=0.3812,Z=0.0805,W=0.9204) ...(truncated)"
    WeightMinimum=0
    Weight=@n4.OutputParam
   )
   n5 = &RigUnit_SphericalPoseReader(
    ActiveRegionScaleFactors=(PositiveWidth=1,NegativeWidth=1,PositiveHeight=1,NegativeHeight=1)
    ActiveRegionSize=0.05
    Debug=(bDrawDebug=false,bDraw2D=False,bDrawLocalAxes=False,DebugScale=25,DebugSegments=20,DebugThickness=0.25)
    DriverAxis=(X=-1,Y=0,Z=0)
    DriverItem=(Type=Bone,Name="thigh_l")
    FalloffRegionScaleFactors=(PositiveWidth=1,NegativeWidth=1,PositiveHeight=0.9,NegativeHeight=0.9)
    FalloffSize=0.65
    FlipHeightScaling=False
    FlipWidthScaling=False
    OptionalParentItem=(Type=Bone,Name="None")
    RotationOffset=(X=90,Y=0,Z=-30)
    ExecutePin=@RigVMFunction_Sequence.B
   )
   > &RigUnit_ModifyTransforms(
    ItemToModify="((Item=(Type=Bone,Name=\"base_skirt_front_l\"),Transform=(Rotation=(X=-0.183,Y=-0.183,Z=-0.683,W=0.683),Translation=(X=-10,Y=0,Z=0),Scale3D=(X=1,Y=1,Z=1))),(Item=(Type=Bone,Name=\"base_skirt_side_l\"),Transform=(Rotation=(X=0.1116,Y=-0.633,Z=0.133,W=0.7544),Tr ...(truncated)"
    WeightMinimum=0
    Weight=@n5.OutputParam
   )
   > n6 = &RigUnit_SphericalPoseReader(
    ActiveRegionScaleFactors=(PositiveWidth=1,NegativeWidth=1,PositiveHeight=1.3,NegativeHeight=1)
    ActiveRegionSize=0.05
    Debug=(bDrawDebug=false,bDraw2D=False,bDrawLocalAxes=False,DebugScale=25,DebugSegments=20,DebugThickness=0.25)
    DriverAxis=(X=-1,Y=0,Z=0)
    DriverItem=(Type=Bone,Name="thigh_l")
    FalloffRegionScaleFactors=(PositiveWidth=1,NegativeWidth=1,PositiveHeight=1.3,NegativeHeight=1)
    FalloffSize=0.2
    FlipHeightScaling=False
    FlipWidthScaling=False
    OptionalParentItem=(Type=Bone,Name="None")
    RotationOffset=(X=140,Y=30,Z=0)
   )
   > &RigUnit_ModifyTransforms(
    ItemToModify="((Item=(Type=Bone,Name=\"base_skirt_front_l\"),Transform=(Rotation=(X=0.0249,Y=0.1281,Z=-0.1892,W=0.9732),Translation=(X=-5,Y=0,Z=0),Scale3D=(X=1,Y=1,Z=1))),(Item=(Type=Bone,Name=\"base_skirt_side_l\"),Transform=(Rotation=(X=0,Y=0,Z=0,W=1),Translation=(X=0,Y=0 ...(truncated)"
    WeightMinimum=0
    Weight=@n6.OutputParam
   )
   > n7 = &RigUnit_SphericalPoseReader(
    ActiveRegionScaleFactors=(PositiveWidth=1,NegativeWidth=1,PositiveHeight=1,NegativeHeight=1)
    ActiveRegionSize=0.05
    Debug=(bDrawDebug=false,bDraw2D=False,bDrawLocalAxes=False,DebugScale=25,DebugSegments=20,DebugThickness=0.25)
    DriverAxis=(X=-1,Y=0,Z=0)
    DriverItem=(Type=Bone,Name="thigh_l")
    FalloffRegionScaleFactors=(PositiveWidth=1,NegativeWidth=1,PositiveHeight=0.9,NegativeHeight=0.9)
    FalloffSize=0.65
    FlipHeightScaling=False
    FlipWidthScaling=False
    OptionalParentItem=(Type=Bone,Name="None")
    RotationOffset=(X=180,Y=27,Z=0)
   )
   > &RigUnit_ModifyTransforms(
    ItemToModify="((Item=(Type=Bone,Name=\"offset_skirt_side_l\"),Transform=(Rotation=(X=0,Y=0,Z=-0.766,W=0.6428),Translation=(X=-5,Y=0,Z=0),Scale3D=(X=1,Y=1,Z=1))),(Item=(Type=Bone,Name=\"offset_skirt_front_l\"),Transform=(Rotation=(X=-0.0223,Y=0.532,Z=-0.2006,W=0.8224),Transl ...(truncated)"
    WeightMinimum=0
    Weight=@n7.OutputParam
   )
   > n8 = &RigUnit_SphericalPoseReader(
    ActiveRegionScaleFactors=(PositiveWidth=1,NegativeWidth=1,PositiveHeight=1,NegativeHeight=1)
    ActiveRegionSize=0.05
    Debug=(bDrawDebug=false,bDraw2D=False,bDrawLocalAxes=False,DebugScale=25,DebugSegments=20,DebugThickness=0.25)
    DriverAxis=(X=-1,Y=0,Z=0)
    DriverItem=(Type=Bone,Name="thigh_l")
    FalloffRegionScaleFactors=(PositiveWidth=1,NegativeWidth=1,PositiveHeight=1,NegativeHeight=1)
    FalloffSize=0.55
    FlipHeightScaling=False
    FlipWidthScaling=False
    OptionalParentItem=(Type=Bone,Name="None")
    RotationOffset=(X=-90,Y=0,Z=20)
   )
   > &RigUnit_ModifyTransforms(
    ItemToModify="((Item=(Type=Bone,Name=\"base_skirt_back_l\"),Transform=(Rotation=(X=0.1782,Y=0.1877,Z=-0.6649,W=0.7007),Translation=(X=-2.5,Y=0,Z=0),Scale3D=(X=1,Y=1,Z=1))),(Item=(Type=Bone,Name=\"base_skirt_side_l\"),Transform=(Rotation=(X=-0.0334,Y=0.3812,Z=0.0805,W=0.9204 ...(truncated)"
    WeightMinimum=0
    Weight=@n8.OutputParam
   )
   RigVMFunctionReturnNode(ExecuteContext=@RigVMFunction_Sequence.C)
   &RigUnit_SphericalPoseReader(
    ActiveRegionScaleFactors=(PositiveWidth=1,NegativeWidth=1,PositiveHeight=1,NegativeHeight=1)
    ActiveRegionSize=0.03
    Debug=(bDrawDebug=false,bDraw2D=False,bDrawLocalAxes=False,DebugScale=25,DebugSegments=20,DebugThickness=0.25)
    DriverAxis=(X=-1,Y=0,Z=0)
    DriverItem=(Type=Bone,Name="thigh_l")
    FalloffRegionScaleFactors=(PositiveWidth=1,NegativeWidth=1,PositiveHeight=0.9,NegativeHeight=0.9)
    FalloffSize=0.3
    FlipHeightScaling=true
    FlipWidthScaling=true
    OptionalParentItem=(Type=Bone,Name="None")
    RotationOffset=(X=60,Y=10,Z=0)
    ExecutePin=@RigVMFunction_Sequence.E
   )
   > &RigUnit_ModifyTransforms(
    ItemToModify="((Item=(Type=Bone,Name=\"base_skirt_front\"),Transform=(Rotation=(X=0,Y=-0.2588,Z=0,W=0.9659),Translation=(X=0,Y=0,Z=0),Scale3D=(X=1,Y=1,Z=1))),(Item=(Type=Bone,Name=\"base_skirt_front_l\"),Transform=(Rotation=(X=-0.0064,Y=-0.1477,Z=-0.0431,W=0.9881),Translati ...(truncated)"
    Weight=0
    WeightMinimum=0
   )
   > &RigUnit_SphericalPoseReader(
    ActiveRegionScaleFactors=(PositiveWidth=1,NegativeWidth=1,PositiveHeight=1,NegativeHeight=1)
    ActiveRegionSize=0.03
    Debug=(bDrawDebug=false,bDraw2D=False,bDrawLocalAxes=False,DebugScale=25,DebugSegments=20,DebugThickness=0.25)
    DriverAxis=(X=1,Y=0,Z=0)
    DriverItem=(Type=Bone,Name="thigh_r")
    FalloffRegionScaleFactors=(PositiveWidth=1,NegativeWidth=1,PositiveHeight=0.9,NegativeHeight=0.9)
    FalloffSize=0.3
    FlipHeightScaling=False
    FlipWidthScaling=False
    OptionalParentItem=(Type=Bone,Name="None")
    RotationOffset=(X=-120,Y=10,Z=0)
   )
   > &RigUnit_ModifyTransforms(
    ItemToModify="((Item=(Type=Bone,Name=\"base_skirt_front\"),Transform=(Rotation=(X=0,Y=0.2588,Z=0,W=0.9659),Translation=(X=0,Y=0,Z=0),Scale3D=(X=1,Y=1,Z=1))),(Item=(Type=Bone,Name=\"base_skirt_front_r\"),Transform=(Rotation=(X=-0.0064,Y=-0.1477,Z=-0.0431,W=0.9881),Translatio ...(truncated)"
    Weight=0
    WeightMinimum=0
   )
   &RigUnit_ModifyTransforms(
    ItemToModify=((Item=(Type=Bone,Name="base_skirt_front"),Transform=(Rotation=(X=0,Y=0,Z=-0.7071,W=0.7071),Translation=(X=-10,Y=0,Z=0),Scale3D=(X=1,Y=1,Z=1))))
    WeightMinimum=0
    Weight=&RigVMFunction_MathDoubleMax(B=@n5.OutputParam, A=@n1.OutputParam).Result
    ExecutePin=@RigVMFunction_Sequence.D
   )
   > &RigUnit_ModifyTransforms(
    ItemToModify=((Item=(Type=Bone,Name="base_skirt_back"),Transform=(Rotation=(X=0,Y=0,Z=-0.6088,W=0.7934),Translation=(X=-5,Y=0,Z=0),Scale3D=(X=1,Y=1,Z=1))))
    WeightMinimum=0
    Weight=&RigVMFunction_MathDoubleMax(B=@n8.OutputParam, A=@n4.OutputParam).Result
   )
   n9 = &RigVMFunction_MathDoubleMul(
    A=&RigVMFunction_MathDoubleRemap(
     SourceMaximum=43.5
     SourceMinimum=89
     TargetMaximum=1
     TargetMinimum=0
     Value=&RigUnit_GetRelativeTransformForItem(Child=(Type=Bone,Name="pelvis"), Parent=(Type=Bone,Name="root"), bChildInitial=False, bParentInitial=False).RelativeTransform.Translation.Z
    ).Result
    B=RigVMVariableNode() {variable=isCrouching}.Value
   )
   &RigUnit_ModifyTransforms(
    ItemToModify="((Item=(Type=Bone,Name=\"base_skirt_front_r\"),Transform=(Rotation=(X=0,Y=0,Z=0.0175,W=0.9998),Translation=(X=-3,Y=0,Z=0),Scale3D=(X=1,Y=1,Z=1))),(Item=(Type=Bone,Name=\"base_skirt_front_l\"),Transform=(Rotation=(X=0,Y=0,Z=0.0175,W=0.9998),Translation=(X=3,Y=0 ...(truncated)"
    WeightMinimum=0
    Weight=@n9.Result
    ExecutePin=@RigVMFunction_Sequence.F
   )
   > &RigUnit_ModifyTransforms(
    ItemToModify="((Item=(Type=Bone,Name=\"base_skirt_back\"),Transform=(Rotation=(X=0,Y=0,Z=0,W=1),Translation=(X=0,Y=-1,Z=0),Scale3D=(X=1,Y=1,Z=1))),(Item=(Type=Bone,Name=\"base_skirt_back_l\"),Transform=(Rotation=(X=0,Y=0,Z=0,W=1),Translation=(X=0,Y=-1,Z=0),Scale3D=(X=1,Y=1, ...(truncated)"
    WeightMinimum=0
    Weight=@n9.Result
   )
   n10 = &RigUnit_GetRelativeTransformForItem(Child=(Type=Bone,Name="pelvis"), Parent=(Type=Bone,Name="root"), bChildInitial=False, bParentInitial=False)
   &RigUnit_ModifyTransforms(
    ItemToModify="((Item=(Type=Bone,Name=\"base_skirt_back\"),Transform=(Rotation=(X=-0.0328,Y=-0.0386,Z=0.6867,W=-0.7251),Translation=(X=0,Y=-1,Z=0),Scale3D=(X=1,Y=1,Z=1))),(Item=(Type=Bone,Name=\"base_skirt_back_l\"),Transform=(Rotation=(X=0.0174,Y=0.033,Z=0.6356,W=-0.7711),T ...(truncated)"
    WeightMinimum=0
    Weight=&RigVMFunction_MathDoubleMul(
     A=&RigVMFunction_MathDoubleRemap(SourceMaximum=43.5, SourceMinimum=89, TargetMaximum=1, TargetMinimum=0, Value=@n10.RelativeTransform.Translation.Z).Result
     B=RigVMVariableNode() {variable=isSliding}.Value
    ).Result
    ExecutePin=@RigVMFunction_Sequence.G
   )
   n11 = &RigUnit_GetRelativeTransformForItem(Child=(Type=Bone,Name="pelvis"), Parent=(Type=Bone,Name="root"), bChildInitial=False, bParentInitial=False)
   &RigUnit_ModifyTransforms(
    ItemToModify="((Item=(Type=Bone,Name=\"base_skirt_back\"),Transform=(Rotation=(X=0,Y=0,Z=-0.0909,W=0.9959),Translation=(X=0,Y=0,Z=0),Scale3D=(X=1,Y=1,Z=1))),(Item=(Type=Bone,Name=\"base_skirt_back_l\"),Transform=(Rotation=(X=0,Y=0,Z=0,W=1),Translation=(X=0,Y=0,Z=0),Scale3D= ...(truncated)"
    WeightMinimum=0
    Weight=&RigVMFunction_MathDoubleMul(
     A=&RigVMFunction_MathDoubleRemap(SourceMaximum=43.5, SourceMinimum=89, TargetMaximum=1, TargetMinimum=0, Value=@n11.RelativeTransform.Translation.Z).Result
     B=RigVMVariableNode() {variable=isBaseBallSliding}.Value
    ).Result
    ExecutePin=@RigVMFunction_Sequence.H
   )
  ]}
 New Function_ContainedGraph {nodes [
   RigVMFunctionEntryNode()
   > RigVMFunction_Sequence = RigVMAggregateNode()
   n1 = &RigUnit_SphericalPoseReader(
    ActiveRegionScaleFactors=(PositiveWidth=1,NegativeWidth=1,PositiveHeight=1,NegativeHeight=1)
    ActiveRegionSize=0.05
    Debug=(bDrawDebug=false,bDraw2D=False,bDrawLocalAxes=False,DebugScale=25,DebugSegments=20,DebugThickness=0.25)
    DriverAxis=(X=1,Y=0,Z=0)
    DriverItem=(Type=Bone,Name="thigh_r")
    FalloffRegionScaleFactors=(PositiveWidth=1,NegativeWidth=1,PositiveHeight=0.9,NegativeHeight=0.9)
    FalloffSize=0.65
    FlipHeightScaling=false
    FlipWidthScaling=false
    OptionalParentItem=(Type=Bone,Name="None")
    RotationOffset=(X=-90,Y=0,Z=-30)
    ExecutePin=@RigVMFunction_Sequence.A
   )
   RigVMFunction_AnimEvalRichCurve = &RigVMFunction_AnimEvalRichCurve(
    Curve="(EditorCurveData=(Keys=((InterpMode=RCIM_Cubic,TangentMode=RCTM_User,TangentWeightMode=RCTWM_WeightedNone,Time=0,Value=0,ArriveTangent=0.8889,ArriveTangentWeight=0,LeaveTangent=0.8889,LeaveTangentWeight=0),(InterpMode=RCIM_Cubic,TangentMode=RCTM_User,Tange ...(truncated)"
    SourceMinimum=0
    TargetMinimum=0
    Value=@n1.OutputParam
   )
   > &RigUnit_ModifyTransforms(
    ItemToModify="((Item=(Type=Bone,Name=\"base_skirt_front_r\"),Transform=(Rotation=(X=-0.1378,Y=-0.1057,Z=-0.7813,W=0.5995),Translation=(X=0,Y=0,Z=0),Scale3D=(X=1,Y=1,Z=1))),(Item=(Type=Bone,Name=\"base_skirt_side_r\"),Transform=(Rotation=(X=0.0892,Y=-0.6772,Z=-0.0892,W=0.724 ...(truncated)"
    WeightMinimum=0
    Weight=@RigVMFunction_AnimEvalRichCurve.Result
   )
   > n2 = &RigUnit_SphericalPoseReader(
    ActiveRegionScaleFactors=(PositiveWidth=1,NegativeWidth=1,PositiveHeight=1,NegativeHeight=1.3)
    ActiveRegionSize=0.05
    Debug=(bDrawDebug=false,bDraw2D=False,bDrawLocalAxes=False,DebugScale=25,DebugSegments=20,DebugThickness=0.25)
    DriverAxis=(X=1,Y=0,Z=0)
    DriverItem=(Type=Bone,Name="thigh_r")
    FalloffRegionScaleFactors=(PositiveWidth=1,NegativeWidth=1,PositiveHeight=1,NegativeHeight=1.3)
    FalloffSize=0.2
    FlipHeightScaling=true
    FlipWidthScaling=true
    OptionalParentItem=(Type=Bone,Name="None")
    RotationOffset=(X=-40,Y=30,Z=0)
   )
   n3 = &RigVMFunction_AnimEvalRichCurve(
    Curve="(EditorCurveData=(Keys=((InterpMode=RCIM_Cubic,TangentMode=RCTM_User,TangentWeightMode=RCTWM_WeightedNone,Time=0,Value=0,ArriveTangent=0.8889,ArriveTangentWeight=0,LeaveTangent=0.8889,LeaveTangentWeight=0),(InterpMode=RCIM_Cubic,TangentMode=RCTM_User,Tange ...(truncated)"
    SourceMinimum=0
    TargetMinimum=0
    Value=@n2.OutputParam
   )
   > &RigUnit_ModifyTransforms(
    ItemToModify="((Item=(Type=Bone,Name=\"base_skirt_front_r\"),Transform=(Rotation=(X=0,Y=0,Z=-0.342,W=0.9397),Translation=(X=0,Y=0,Z=0),Scale3D=(X=1,Y=1,Z=1))),(Item=(Type=Bone,Name=\"base_skirt_side_r\"),Transform=(Rotation=(X=0,Y=0,Z=0,W=1),Translation=(X=0,Y=0,Z=0),Scale3 ...(truncated)"
    WeightMinimum=0
    Weight=@n3.Result
   )
   > n4 = &RigUnit_SphericalPoseReader(
    ActiveRegionScaleFactors=(PositiveWidth=1,NegativeWidth=1,PositiveHeight=1.3,NegativeHeight=1)
    ActiveRegionSize=0.05
    Debug=(bDrawDebug=True,bDraw2D=False,bDrawLocalAxes=True,DebugScale=25,DebugSegments=20,DebugThickness=0.25)
    DriverAxis=(X=1,Y=0,Z=0)
    DriverItem=(Type=Bone,Name="thigh_r")
    FalloffRegionScaleFactors=(PositiveWidth=1,NegativeWidth=1,PositiveHeight=1.3,NegativeHeight=1)
    FalloffSize=0.2
    FlipHeightScaling=False
    FlipWidthScaling=False
    OptionalParentItem=(Type=Bone,Name="None")
    RotationOffset=(X=-110,Y=-30,Z=45)
   )
   n5 = &RigVMFunction_AnimEvalRichCurve(
    Curve="(EditorCurveData=(Keys=((InterpMode=RCIM_Cubic,TangentMode=RCTM_User,TangentWeightMode=RCTWM_WeightedNone,Time=0,Value=0,ArriveTangent=0.8889,ArriveTangentWeight=0,LeaveTangent=0.8889,LeaveTangentWeight=0),(InterpMode=RCIM_Cubic,TangentMode=RCTM_User,Tange ...(truncated)"
    SourceMinimum=0
    TargetMinimum=0
    Value=@n4.OutputParam
   )
   > &RigUnit_ModifyTransforms(
    ItemToModify="((Item=(Type=Bone,Name=\"base_skirt_front_r\"),Transform=(Rotation=(X=-0.0258,Y=-0.0959,Z=-0.0003,W=0.9951),Translation=(X=0,Y=0,Z=0),Scale3D=(X=1,Y=1,Z=1))),(Item=(Type=Bone,Name=\"base_skirt_side_r\"),Transform=(Rotation=(X=0,Y=0,Z=0.0057,W=1),Translation=(X ...(truncated)"
    WeightMinimum=0
    Weight=@n5.Result
   )
   > n6 = &RigUnit_SphericalPoseReader(
    ActiveRegionScaleFactors=(PositiveWidth=1,NegativeWidth=1,PositiveHeight=1,NegativeHeight=1)
    ActiveRegionSize=0.05
    Debug=(bDrawDebug=false,bDraw2D=False,bDrawLocalAxes=False,DebugScale=25,DebugSegments=20,DebugThickness=0.25)
    DriverAxis=(X=1,Y=0,Z=0)
    DriverItem=(Type=Bone,Name="thigh_r")
    FalloffRegionScaleFactors=(PositiveWidth=1,NegativeWidth=1,PositiveHeight=0.9,NegativeHeight=0.9)
    FalloffSize=0.65
    FlipHeightScaling=true
    FlipWidthScaling=true
    OptionalParentItem=(Type=Bone,Name="None")
    RotationOffset=(X=0,Y=27,Z=0)
   )
   n7 = &RigVMFunction_AnimEvalRichCurve(
    Curve="(EditorCurveData=(Keys=((InterpMode=RCIM_Cubic,TangentMode=RCTM_User,TangentWeightMode=RCTWM_WeightedNone,Time=0,Value=0,ArriveTangent=0.8889,ArriveTangentWeight=0,LeaveTangent=0.8889,LeaveTangentWeight=0),(InterpMode=RCIM_Cubic,TangentMode=RCTM_User,Tange ...(truncated)"
    SourceMinimum=0
    TargetMinimum=0
    Value=@n6.OutputParam
   )
   > &RigUnit_ModifyTransforms(
    ItemToModify="((Item=(Type=Bone,Name=\"offset_skirt_side_r\"),Transform=(Rotation=(X=0,Y=0,Z=-0.7373,W=0.6756),Translation=(X=0,Y=0,Z=0),Scale3D=(X=1,Y=1,Z=1))),(Item=(Type=Bone,Name=\"offset_skirt_front_r\"),Transform=(Rotation=(X=-0.1648,Y=0.6092,Z=-0.0804,W=0.7715),Trans ...(truncated)"
    WeightMinimum=0
    Weight=@n7.Result
   )
   > n8 = &RigUnit_SphericalPoseReader(
    ActiveRegionScaleFactors=(PositiveWidth=1,NegativeWidth=1,PositiveHeight=1,NegativeHeight=1)
    ActiveRegionSize=0.05
    Debug=(bDrawDebug=false,bDraw2D=False,bDrawLocalAxes=False,DebugScale=25,DebugSegments=20,DebugThickness=0.25)
    DriverAxis=(X=1,Y=0,Z=0)
    DriverItem=(Type=Bone,Name="thigh_r")
    FalloffRegionScaleFactors=(PositiveWidth=1,NegativeWidth=1,PositiveHeight=1,NegativeHeight=1)
    FalloffSize=0.55
    FlipHeightScaling=False
    FlipWidthScaling=False
    OptionalParentItem=(Type=Bone,Name="None")
    RotationOffset=(X=90,Y=0,Z=20)
   )
   n9 = &RigVMFunction_AnimEvalRichCurve(
    Curve="(EditorCurveData=(Keys=((InterpMode=RCIM_Cubic,TangentMode=RCTM_User,TangentWeightMode=RCTWM_WeightedNone,Time=0,Value=0,ArriveTangent=0.8889,ArriveTangentWeight=0,LeaveTangent=0.8889,LeaveTangentWeight=0),(InterpMode=RCIM_Cubic,TangentMode=RCTM_User,Tange ...(truncated)"
    SourceMinimum=0
    TargetMinimum=0
    Value=@n8.OutputParam
   )
   > &RigUnit_ModifyTransforms(
    ItemToModify="((Item=(Type=Bone,Name=\"base_skirt_back_r\"),Transform=(Rotation=(X=0.183,Y=0.183,Z=-0.683,W=0.683),Translation=(X=0,Y=0,Z=0),Scale3D=(X=1,Y=1,Z=1))),(Item=(Type=Bone,Name=\"base_skirt_side_r\"),Transform=(Rotation=(X=-0.0334,Y=0.3812,Z=0.0805,W=0.9204),Trans ...(truncated)"
    WeightMinimum=0
    Weight=@n9.Result
   )
   n10 = &RigUnit_SphericalPoseReader(
    ActiveRegionScaleFactors=(PositiveWidth=1,NegativeWidth=1,PositiveHeight=1,NegativeHeight=1)
    ActiveRegionSize=0.05
    Debug=(bDrawDebug=false,bDraw2D=False,bDrawLocalAxes=False,DebugScale=25,DebugSegments=20,DebugThickness=0.25)
    DriverAxis=(X=-1,Y=0,Z=0)
    DriverItem=(Type=Bone,Name="thigh_l")
    FalloffRegionScaleFactors=(PositiveWidth=1,NegativeWidth=1,PositiveHeight=0.9,NegativeHeight=0.9)
    FalloffSize=0.65
    FlipHeightScaling=False
    FlipWidthScaling=False
    OptionalParentItem=(Type=Bone,Name="None")
    RotationOffset=(X=90,Y=0,Z=-30)
    ExecutePin=@RigVMFunction_Sequence.B
   )
   n11 = &RigVMFunction_AnimEvalRichCurve(
    Curve="(EditorCurveData=(Keys=((InterpMode=RCIM_Cubic,TangentMode=RCTM_User,TangentWeightMode=RCTWM_WeightedNone,Time=0,Value=0,ArriveTangent=0.8889,ArriveTangentWeight=0,LeaveTangent=0.8889,LeaveTangentWeight=0),(InterpMode=RCIM_Cubic,TangentMode=RCTM_User,Tange ...(truncated)"
    SourceMinimum=0
    TargetMinimum=0
    Value=@n10.OutputParam
   )
   > &RigUnit_ModifyTransforms(
    ItemToModify="((Item=(Type=Bone,Name=\"base_skirt_front_l\"),Transform=(Rotation=(X=-0.1378,Y=-0.1057,Z=-0.7813,W=0.5995),Translation=(X=0,Y=0,Z=0),Scale3D=(X=1,Y=1,Z=1))),(Item=(Type=Bone,Name=\"base_skirt_side_l\"),Transform=(Rotation=(X=0.0892,Y=-0.6772,Z=-0.0892,W=0.724 ...(truncated)"
    WeightMinimum=0
    Weight=@n11.Result
   )
   > n12 = &RigUnit_SphericalPoseReader(
    ActiveRegionScaleFactors=(PositiveWidth=1,NegativeWidth=1,PositiveHeight=1.3,NegativeHeight=1)
    ActiveRegionSize=0.05
    Debug=(bDrawDebug=false,bDraw2D=False,bDrawLocalAxes=False,DebugScale=25,DebugSegments=20,DebugThickness=0.25)
    DriverAxis=(X=-1,Y=0,Z=0)
    DriverItem=(Type=Bone,Name="thigh_l")
    FalloffRegionScaleFactors=(PositiveWidth=1,NegativeWidth=1,PositiveHeight=1.3,NegativeHeight=1)
    FalloffSize=0.2
    FlipHeightScaling=False
    FlipWidthScaling=False
    OptionalParentItem=(Type=Bone,Name="None")
    RotationOffset=(X=140,Y=30,Z=0)
   )
   n13 = &RigVMFunction_AnimEvalRichCurve(
    Curve="(EditorCurveData=(Keys=((InterpMode=RCIM_Cubic,TangentMode=RCTM_User,TangentWeightMode=RCTWM_WeightedNone,Time=0,Value=0,ArriveTangent=0.8889,ArriveTangentWeight=0,LeaveTangent=0.8889,LeaveTangentWeight=0),(InterpMode=RCIM_Cubic,TangentMode=RCTM_User,Tange ...(truncated)"
    SourceMinimum=0
    TargetMinimum=0
    Value=@n12.OutputParam
   )
   > &RigUnit_ModifyTransforms(
    ItemToModify="((Item=(Type=Bone,Name=\"base_skirt_front_l\"),Transform=(Rotation=(X=0,Y=0,Z=-0.342,W=0.9397),Translation=(X=0,Y=0,Z=0),Scale3D=(X=1,Y=1,Z=1))),(Item=(Type=Bone,Name=\"base_skirt_side_l\"),Transform=(Rotation=(X=0,Y=0,Z=0,W=1),Translation=(X=0,Y=0,Z=0),Scale3 ...(truncated)"
    WeightMinimum=0
    Weight=@n13.Result
   )
   > n14 = &RigUnit_SphericalPoseReader(
    ActiveRegionScaleFactors=(PositiveWidth=1,NegativeWidth=1,PositiveHeight=1.3,NegativeHeight=1)
    ActiveRegionSize=0.05
    Debug=(bDrawDebug=True,bDraw2D=False,bDrawLocalAxes=True,DebugScale=25,DebugSegments=20,DebugThickness=0.25)
    DriverAxis=(X=-1,Y=0,Z=0)
    DriverItem=(Type=Bone,Name="thigh_l")
    FalloffRegionScaleFactors=(PositiveWidth=1,NegativeWidth=1,PositiveHeight=1.3,NegativeHeight=1)
    FalloffSize=0.2
    FlipHeightScaling=False
    FlipWidthScaling=False
    OptionalParentItem=(Type=Bone,Name="None")
    RotationOffset=(X=70,Y=-30,Z=45)
   )
   n15 = &RigVMFunction_AnimEvalRichCurve(
    Curve="(EditorCurveData=(Keys=((InterpMode=RCIM_Cubic,TangentMode=RCTM_User,TangentWeightMode=RCTWM_WeightedNone,Time=0,Value=0,ArriveTangent=0.8889,ArriveTangentWeight=0,LeaveTangent=0.8889,LeaveTangentWeight=0),(InterpMode=RCIM_Cubic,TangentMode=RCTM_User,Tange ...(truncated)"
    SourceMinimum=0
    TargetMinimum=0
    Value=@n14.OutputParam
   )
   > &RigUnit_ModifyTransforms(
    ItemToModify="((Item=(Type=Bone,Name=\"base_skirt_front_l\"),Transform=(Rotation=(X=0.0263,Y=-0.0958,Z=0.0047,W=0.995),Translation=(X=0,Y=0,Z=0),Scale3D=(X=1,Y=1,Z=1))),(Item=(Type=Bone,Name=\"base_skirt_side_l\"),Transform=(Rotation=(X=0,Y=0,Z=0.0057,W=1),Translation=(X=0, ...(truncated)"
    WeightMinimum=0
    Weight=@n15.Result
   )
   > n16 = &RigUnit_SphericalPoseReader(
    ActiveRegionScaleFactors=(PositiveWidth=1,NegativeWidth=1,PositiveHeight=1,NegativeHeight=1)
    ActiveRegionSize=0.05
    Debug=(bDrawDebug=false,bDraw2D=False,bDrawLocalAxes=False,DebugScale=25,DebugSegments=20,DebugThickness=0.25)
    DriverAxis=(X=-1,Y=0,Z=0)
    DriverItem=(Type=Bone,Name="thigh_l")
    FalloffRegionScaleFactors=(PositiveWidth=1,NegativeWidth=1,PositiveHeight=0.9,NegativeHeight=0.9)
    FalloffSize=0.65
    FlipHeightScaling=False
    FlipWidthScaling=False
    OptionalParentItem=(Type=Bone,Name="None")
    RotationOffset=(X=180,Y=27,Z=0)
   )
   n17 = &RigVMFunction_AnimEvalRichCurve(
    Curve="(EditorCurveData=(Keys=((InterpMode=RCIM_Cubic,TangentMode=RCTM_User,TangentWeightMode=RCTWM_WeightedNone,Time=0,Value=0,ArriveTangent=0.8889,ArriveTangentWeight=0,LeaveTangent=0.8889,LeaveTangentWeight=0),(InterpMode=RCIM_Cubic,TangentMode=RCTM_User,Tange ...(truncated)"
    SourceMinimum=0
    TargetMinimum=0
    Value=@n16.OutputParam
   )
   > &RigUnit_ModifyTransforms(
    ItemToModify="((Item=(Type=Bone,Name=\"offset_skirt_side_l\"),Transform=(Rotation=(X=0,Y=0,Z=-0.7373,W=0.6756),Translation=(X=0,Y=0,Z=0),Scale3D=(X=1,Y=1,Z=1))),(Item=(Type=Bone,Name=\"offset_skirt_front_l\"),Transform=(Rotation=(X=-0.1648,Y=0.6092,Z=-0.0804,W=0.7715),Trans ...(truncated)"
    WeightMinimum=0
    Weight=@n17.Result
   )
   > n18 = &RigUnit_SphericalPoseReader(
    ActiveRegionScaleFactors=(PositiveWidth=1,NegativeWidth=1,PositiveHeight=1,NegativeHeight=1)
    ActiveRegionSize=0.05
    Debug=(bDrawDebug=false,bDraw2D=False,bDrawLocalAxes=False,DebugScale=25,DebugSegments=20,DebugThickness=0.25)
    DriverAxis=(X=-1,Y=0,Z=0)
    DriverItem=(Type=Bone,Name="thigh_l")
    FalloffRegionScaleFactors=(PositiveWidth=1,NegativeWidth=1,PositiveHeight=1,NegativeHeight=1)
    FalloffSize=0.55
    FlipHeightScaling=False
    FlipWidthScaling=False
    OptionalParentItem=(Type=Bone,Name="None")
    RotationOffset=(X=-90,Y=0,Z=20)
   )
   n19 = &RigVMFunction_AnimEvalRichCurve(
    Curve="(EditorCurveData=(Keys=((InterpMode=RCIM_Cubic,TangentMode=RCTM_User,TangentWeightMode=RCTWM_WeightedNone,Time=0,Value=0,ArriveTangent=0.8889,ArriveTangentWeight=0,LeaveTangent=0.8889,LeaveTangentWeight=0),(InterpMode=RCIM_Cubic,TangentMode=RCTM_User,Tange ...(truncated)"
    SourceMinimum=0
    TargetMinimum=0
    Value=@n18.OutputParam
   )
   > &RigUnit_ModifyTransforms(
    ItemToModify="((Item=(Type=Bone,Name=\"base_skirt_back_l\"),Transform=(Rotation=(X=0.183,Y=0.183,Z=-0.683,W=0.683),Translation=(X=0,Y=0,Z=0),Scale3D=(X=1,Y=1,Z=1))),(Item=(Type=Bone,Name=\"base_skirt_side_l\"),Transform=(Rotation=(X=-0.0334,Y=0.3812,Z=0.0805,W=0.9204),Trans ...(truncated)"
    WeightMinimum=0
    Weight=@n19.Result
   )
   RigVMFunctionReturnNode(ExecuteContext=@RigVMFunction_Sequence.C)
   n20 = &RigVMFunction_AnimEvalRichCurve(
    Curve="(EditorCurveData=(Keys=((InterpMode=RCIM_Cubic,TangentMode=RCTM_User,TangentWeightMode=RCTWM_WeightedNone,Time=0,Value=0,ArriveTangent=0.8889,ArriveTangentWeight=0,LeaveTangent=0.8889,LeaveTangentWeight=0),(InterpMode=RCIM_Cubic,TangentMode=RCTM_User,Tange ...(truncated)"
    SourceMinimum=0
    TargetMinimum=0
    Value=&RigVMFunction_MathDoubleMax(B=@n10.OutputParam, A=@n1.OutputParam).Result
   )
   &RigUnit_ModifyTransforms(
    ItemToModify=((Item=(Type=Bone,Name="base_skirt_front"),Transform=(Rotation=(X=0,Y=0,Z=-0.7071,W=0.7071),Translation=(X=0,Y=0,Z=0),Scale3D=(X=1,Y=1,Z=1))))
    WeightMinimum=0
    Weight=@n20.Result
    ExecutePin=@RigVMFunction_Sequence.D
   )
   n21 = &RigVMFunction_AnimEvalRichCurve(
    Curve="(EditorCurveData=(Keys=((InterpMode=RCIM_Cubic,TangentMode=RCTM_User,TangentWeightMode=RCTWM_WeightedNone,Time=0,Value=0,ArriveTangent=0.8889,ArriveTangentWeight=0,LeaveTangent=0.8889,LeaveTangentWeight=0),(InterpMode=RCIM_Cubic,TangentMode=RCTM_User,Tange ...(truncated)"
    SourceMinimum=0
    TargetMinimum=0
    Value=&RigVMFunction_MathDoubleMax(A=@n8.OutputParam, B=@n18.OutputParam).Result
   )
   > &RigUnit_ModifyTransforms(
    ItemToModify=((Item=(Type=Bone,Name="base_skirt_back"),Transform=(Rotation=(X=0,Y=0,Z=-0.6428,W=0.766),Translation=(X=0,Y=0,Z=0),Scale3D=(X=1,Y=1,Z=1))))
    WeightMinimum=0
    Weight=@n21.Result
   )
   n22 = &RigUnit_SphericalPoseReader(
    ActiveRegionScaleFactors=(PositiveWidth=1,NegativeWidth=1,PositiveHeight=1,NegativeHeight=1)
    ActiveRegionSize=0.1
    Debug=(bDrawDebug=false,bDraw2D=False,bDrawLocalAxes=False,DebugScale=25,DebugSegments=20,DebugThickness=0.25)
    DriverAxis=(X=-1,Y=0,Z=0)
    DriverItem=(Type=Bone,Name="thigh_l")
    FalloffRegionScaleFactors=(PositiveWidth=1,NegativeWidth=1,PositiveHeight=0.9,NegativeHeight=0.9)
    FalloffSize=0.3
    FlipHeightScaling=true
    FlipWidthScaling=true
    OptionalParentItem=(Type=Bone,Name="None")
    RotationOffset=(X=60,Y=10,Z=0)
    ExecutePin=@RigVMFunction_Sequence.E
   )
   > &RigUnit_ModifyTransforms(
    ItemToModify="((Item=(Type=Bone,Name=\"base_skirt_front\"),Transform=(Rotation=(X=0.0194,Y=-0.2613,Z=-0.0729,W=0.9623),Translation=(X=0,Y=0,Z=0),Scale3D=(X=1,Y=1,Z=1))),(Item=(Type=Bone,Name=\"base_skirt_front_l\"),Transform=(Rotation=(X=-0.0011,Y=-0.2086,Z=-0.0816,W=0.9746 ...(truncated)"
    WeightMinimum=0
    Weight=&RigVMFunction_MathDoubleMul(A=@n22.OutputParam, B=RigVMVariableNode() {variable=isSprinting}.Value).Result
   )
   > n23 = &RigUnit_SphericalPoseReader(
    ActiveRegionScaleFactors=(PositiveWidth=1,NegativeWidth=1,PositiveHeight=1,NegativeHeight=1)
    ActiveRegionSize=0.1
    Debug=(bDrawDebug=false,bDraw2D=False,bDrawLocalAxes=False,DebugScale=25,DebugSegments=20,DebugThickness=0.25)
    DriverAxis=(X=1,Y=0,Z=0)
    DriverItem=(Type=Bone,Name="thigh_r")
    FalloffRegionScaleFactors=(PositiveWidth=1,NegativeWidth=1,PositiveHeight=0.9,NegativeHeight=0.9)
    FalloffSize=0.3
    FlipHeightScaling=False
    FlipWidthScaling=False
    OptionalParentItem=(Type=Bone,Name="None")
    RotationOffset=(X=-120,Y=10,Z=0)
   )
   > &RigUnit_ModifyTransforms(
    ItemToModify="((Item=(Type=Bone,Name=\"base_skirt_front\"),Transform=(Rotation=(X=0.0645,Y=0.2539,Z=-0.0954,W=0.9604),Translation=(X=0,Y=0,Z=0),Scale3D=(X=1,Y=1,Z=1))),(Item=(Type=Bone,Name=\"base_skirt_front_r\"),Transform=(Rotation=(X=-0.0011,Y=-0.2086,Z=-0.0816,W=0.9746) ...(truncated)"
    WeightMinimum=0
    Weight=&RigVMFunction_MathDoubleMul(A=@n23.OutputParam, B=RigVMVariableNode() {variable=isSprinting}.Value).Result
   )
   &RigUnit_ModifyTransforms(
    ItemToModify="((Item=(Type=Bone,Name=\"base_skirt_back_l\"),Transform=(Rotation=(X=0,Y=0,Z=-0.0262,W=0.9997),Translation=(X=0,Y=0,Z=0),Scale3D=(X=1,Y=1,Z=1))),(Item=(Type=Bone,Name=\"base_skirt_back_r\"),Transform=(Rotation=(X=0,Y=0,Z=-0.0262,W=0.9997),Translation=(X=0,Y=0, ...(truncated)"
    WeightMinimum=0
    Weight=/Script/RigVM.RigVMFunction_AnimEasing(
     SourceMaximum=1
     SourceMinimum=0
     TargetMaximum=1
     TargetMinimum=0
     Type=CubicEaseOut
     Value=RigVMVariableNode() {variable=isSkyDiving}.Value
    ) {is=struct RigVMUnitNode}.Result
    ExecutePin=@RigVMFunction_Sequence.I
   )
   n24 = &RigUnit_GetRelativeTransformForItem(Child=(Type=Bone,Name="spine_01"), Parent=(Type=Bone,Name="pelvis"), bChildInitial=false, bParentInitial=false)
   n25 = &RigVMFunction_MathDoubleRemap(
    SourceMaximum=50
    SourceMinimum=-1
    TargetMaximum=1
    TargetMinimum=0
    Value=&RigVMFunction_MathQuaternionToEuler(
     RotationOrder=ZYX
     Value=&RigVMFunction_MathQuaternionSwingTwist(TwistAxis=(X=1,Y=0,Z=0), Input=@n24.RelativeTransform.Rotation).Swing
    ).Result.Z
   )
   n26 = RigVMVariableNode() {variable=isRailGringind}
   &RigUnit_ModifyTransforms(
    ItemToModify="((Item=(Type=Bone,Name=\"base_skirt_front_l\"),Transform=(Rotation=(X=-0.0027,Y=-0.061,Z=-0.0435,W=0.9972),Translation=(X=0,Y=0,Z=0),Scale3D=(X=1,Y=1,Z=1))),(Item=(Type=Bone,Name=\"base_skirt_side_l\"),Transform=(Rotation=(X=0,Y=0,Z=0,W=1),Translation=(X=0,Y=0 ...(truncated)"
    WeightMinimum=0
    Weight=&RigVMFunction_AnimEasing(
     SourceMaximum=1
     SourceMinimum=0
     TargetMaximum=1
     TargetMinimum=0
     Type=CubicEaseOut
     Value=&RigVMFunction_MathDoubleMul(B=@n25.Result, A=@n26.Value).Result
    ).Result
    ExecutePin=@RigVMFunction_Sequence.J
   )
   n27 = &RigUnit_SphericalPoseReader(
    ActiveRegionScaleFactors=(PositiveWidth=1.5,NegativeWidth=2.2,PositiveHeight=2.2,NegativeHeight=1)
    ActiveRegionSize=0.05
    Debug=(bDrawDebug=false,bDraw2D=False,bDrawLocalAxes=False,DebugScale=25,DebugSegments=20,DebugThickness=0.25)
    DriverAxis=(X=-1,Y=0,Z=0)
    DriverItem=(Type=Bone,Name="thigh_l")
    FalloffRegionScaleFactors=(PositiveWidth=0.6,NegativeWidth=0.6,PositiveHeight=0.6,NegativeHeight=0.3)
    FalloffSize=0.3
    FlipHeightScaling=true
    FlipWidthScaling=true
    OptionalParentItem=(Type=Bone,Name="None")
    RotationOffset=(X=70,Y=0,Z=0)
    ExecutePin=@RigVMFunction_Sequence.K
   )
   > &RigUnit_ModifyTransforms(
    ItemToModify="((Item=(Type=Bone,Name=\"base_skirt_front_l\"),Transform=(Rotation=(X=-0.0049,Y=-0.0696,Z=-0.0696,W=0.9951),Translation=(X=0,Y=0,Z=0),Scale3D=(X=1,Y=1,Z=1))),(Item=(Type=Bone,Name=\"base_skirt_front\"),Transform=(Rotation=(X=-0.0076,Y=-0.0868,Z=-0.0868,W=0.992 ...(truncated)"
    WeightMinimum=0
    Weight=&RigVMFunction_MathDoubleMul(A=@n27.OutputParam, B=RigVMVariableNode() {variable=isCrouching}.Value).Result
   )
   > n28 = &RigUnit_SphericalPoseReader(
    ActiveRegionScaleFactors=(PositiveWidth=1.5,NegativeWidth=2.2,PositiveHeight=1,NegativeHeight=2.2)
    ActiveRegionSize=0.05
    Debug=(bDrawDebug=false,bDraw2D=False,bDrawLocalAxes=False,DebugScale=25,DebugSegments=20,DebugThickness=0.25)
    DriverAxis=(X=1,Y=0,Z=0)
    DriverItem=(Type=Bone,Name="thigh_r")
    FalloffRegionScaleFactors=(PositiveWidth=0.6,NegativeWidth=0.6,PositiveHeight=0.3,NegativeHeight=0.6)
    FalloffSize=0.3
    FlipHeightScaling=False
    FlipWidthScaling=False
    OptionalParentItem=(Type=Bone,Name="None")
    RotationOffset=(X=-110,Y=0,Z=0)
   )
   > &RigUnit_ModifyTransforms(
    ItemToModify="((Item=(Type=Bone,Name=\"base_skirt_front_r\"),Transform=(Rotation=(X=-0.0049,Y=-0.0696,Z=-0.0696,W=0.9951),Translation=(X=0,Y=0,Z=0),Scale3D=(X=1,Y=1,Z=1))),(Item=(Type=Bone,Name=\"base_skirt_front\"),Transform=(Rotation=(X=0.0125,Y=0.1299,Z=-0.095,W=0.9869), ...(truncated)"
    WeightMinimum=0
    Weight=&RigVMFunction_MathDoubleMul(A=@n28.OutputParam, B=RigVMVariableNode() {variable=isCrouching}.Value).Result
   )
   n29 = &RigVMFunction_MathDoubleMul(
    A=&RigVMFunction_MathDoubleRemap(
     SourceMaximum=55
     SourceMinimum=89
     TargetMaximum=1
     TargetMinimum=0
     Value=&RigUnit_GetRelativeTransformForItem(Child=(Type=Bone,Name="pelvis"), Parent=(Type=Bone,Name="root"), bChildInitial=False, bParentInitial=False).RelativeTransform.Translation.Z
    ).Result
    B=RigVMVariableNode() {variable=isCrouching}.Value
   )
   &RigUnit_ModifyTransforms(
    ItemToModify="((Item=(Type=Bone,Name=\"base_skirt_front_r\"),Transform=(Rotation=(X=0,Y=0,Z=0.0175,W=0.9998),Translation=(X=0,Y=0,Z=0),Scale3D=(X=1,Y=1,Z=1))),(Item=(Type=Bone,Name=\"base_skirt_front_l\"),Transform=(Rotation=(X=0,Y=0,Z=0,W=1),Translation=(X=0,Y=0,Z=0),Scale ...(truncated)"
    WeightMinimum=0
    Weight=@n29.Result
    ExecutePin=@RigVMFunction_Sequence.F
   )
   > &RigUnit_ModifyTransforms(
    ItemToModify="((Item=(Type=Bone,Name=\"base_skirt_back\"),Transform=(Rotation=(X=0,Y=0,Z=0,W=1),Translation=(X=0,Y=0,Z=0),Scale3D=(X=1,Y=1,Z=1))),(Item=(Type=Bone,Name=\"dyn_skirt_back_top_l_02\"),Transform=(Rotation=(X=-0.1137,Y=0.057,Z=0.2946,W=0.9471),Translation=(X=0,Y= ...(truncated)"
    WeightMinimum=0
    Weight=@n29.Result
   )
   n30 = &RigVMFunction_MathDoubleMul(B=1, A=@n29.Result)
   > &RigUnit_ModifyTransforms(
    ItemToModify="((Item=(Type=Bone,Name=\"base_skirt_side_l\"),Transform=(Rotation=(X=-0.0394,Y=0.1819,Z=0.1268,W=0.9743),Translation=(X=0,Y=0,Z=0),Scale3D=(X=1,Y=1,Z=1))),(Item=(Type=Bone,Name=\"dyn_skirt_side_top_l_02\"),Transform=(Rotation=(X=-0.0014,Y=0.0097,Z=0.1412,W=0.9 ...(truncated)"
    WeightMinimum=0
    Weight=@n30.Result
   )
   > &RigUnit_ModifyTransforms(
    ItemToModify="((Item=(Type=Bone,Name=\"base_skirt_front_l\"),Transform=(Rotation=(X=0,Y=0,Z=0,W=1),Translation=(X=-5,Y=0,Z=0),Scale3D=(X=1,Y=1,Z=1))),(Item=(Type=Bone,Name=\"base_skirt_side_l\"),Transform=(Rotation=(X=0,Y=0,Z=0,W=1),Translation=(X=-5,Y=0,Z=0),Scale3D=(X=1,Y ...(truncated)"
    WeightMinimum=0
    Weight=@n30.Result
   )
   > &RigUnit_ModifyTransforms(
    ItemToModify="((Item=(Type=Bone,Name=\"dyn_skirt_front_top_l_02\"),Transform=(Rotation=(X=0,Y=0,Z=0,W=1),Translation=(X=-5,Y=0,Z=0),Scale3D=(X=1,Y=1,Z=1))),(Item=(Type=Bone,Name=\"dyn_skirt_side_top_l_02\"),Transform=(Rotation=(X=0,Y=0,Z=0,W=1),Translation=(X=-5,Y=0,Z=0),Sc ...(truncated)"
    WeightMinimum=0
    Weight=@n30.Result
   )
   > &RigUnit_ModifyTransforms(
    ItemToModify=((Item=(Type=Bone,Name="base_skirt_side_l"),Transform=(Rotation=(X=0,Y=0,Z=-0.0785,W=0.9969),Translation=(X=0,Y=0,Z=0),Scale3D=(X=1,Y=1,Z=1))))
    WeightMinimum=0
    Weight=&RigVMFunction_MathDoubleMul(A=@n30.Result, B=RigVMVariableNode() {variable=isMoving2D}.Value).Result
   )
   n31 = &RigVMFunction_MathDoubleRemap(
    SourceMaximum=43.5
    SourceMinimum=89
    TargetMaximum=1
    TargetMinimum=0
    Value=&RigUnit_GetRelativeTransformForItem(Child=(Type=Bone,Name="pelvis"), Parent=(Type=Bone,Name="root"), bChildInitial=False, bParentInitial=False).RelativeTransform.Translation.Z
   )
   n32 = RigVMVariableNode() {variable=isSliding}
   &RigUnit_ModifyTransforms(
    ItemToModify="((Item=(Type=Bone,Name=\"base_skirt_back\"),Transform=(Rotation=(X=-0.0837,Y=-0.0625,Z=0.5189,W=-0.8484),Translation=(X=0,Y=0,Z=0),Scale3D=(X=1,Y=1,Z=1))),(Item=(Type=Bone,Name=\"base_skirt_back_l\"),Transform=(Rotation=(X=0,Y=0,Z=0.39,W=-0.9208),Translation=( ...(truncated)"
    WeightMinimum=0
    Weight=&RigVMFunction_AnimEasing(
     SourceMaximum=1
     SourceMinimum=0
     TargetMaximum=1
     TargetMinimum=0
     Type=CubicEaseOut
     Value=&RigVMFunction_MathDoubleMul(A=@n31.Result, B=@n32.Value).Result
    ).Result
    ExecutePin=@RigVMFunction_Sequence.G
   )
   n33 = &RigVMFunction_MathDoubleRemap(
    SourceMaximum=43.5
    SourceMinimum=89
    TargetMaximum=1
    TargetMinimum=0
    Value=&RigUnit_GetRelativeTransformForItem(Child=(Type=Bone,Name="pelvis"), Parent=(Type=Bone,Name="root"), bChildInitial=False, bParentInitial=False).RelativeTransform.Translation.Z
   )
   n34 = RigVMVariableNode() {variable=isBaseBallSliding}
   &RigUnit_ModifyTransforms(
    ItemToModify="((Item=(Type=Bone,Name=\"base_skirt_back\"),Transform=(Rotation=(X=-0.2797,Y=0.1285,Z=-0.1086,W=0.9452),Translation=(X=0,Y=0,Z=0),Scale3D=(X=1,Y=1,Z=1))),(Item=(Type=Bone,Name=\"base_skirt_back_l\"),Transform=(Rotation=(X=-0.1915,Y=0.3255,Z=0.031,W=0.9254),Tra ...(truncated)"
    WeightMinimum=0
    Weight=&RigVMFunction_AnimEasing(
     SourceMaximum=1
     SourceMinimum=0
     TargetMaximum=1
     TargetMinimum=0
     Type=ExponentialEaseOut
     Value=&RigVMFunction_MathDoubleMul(A=@n33.Result, B=@n34.Value).Result
    ).Result
    ExecutePin=@RigVMFunction_Sequence.H
   )
   > &RigUnit_ModifyTransforms(
    ItemToModify="((Item=(Type=Bone,Name=\"base_skirt_front_r\"),Transform=(Rotation=(X=0,Y=0,Z=0.2946,W=0.9556),Translation=(X=-18.0295,Y=12.269,Z=6.4312),Scale3D=(X=1,Y=1,Z=1))),(Item=(Type=Bone,Name=\"base_skirt_side_r\"),Transform=(Rotation=(X=0,Y=0,Z=-0.4349,W=-0.9005),Tra ...(truncated)"
    Weight=0
    WeightMinimum=0
   )
   &RigUnit_ModifyTransforms(
    ExecutePin=()
    ItemToModify="((Item=(Type=Bone,Name=\"base_skirt_front_l\"),Transform=(Rotation=(X=0,Y=0,Z=0,W=1),Translation=(X=-5,Y=0,Z=0),Scale3D=(X=1,Y=1,Z=1))),(Item=(Type=Bone,Name=\"base_skirt_side_l\"),Transform=(Rotation=(X=0,Y=0,Z=0,W=1),Translation=(X=-5,Y=0,Z=0),Scale3D=(X=1,Y ...(truncated)"
    Weight=1
    WeightMinimum=0
   )
   > &RigUnit_ModifyTransforms(
    ItemToModify="((Item=(Type=Bone,Name=\"dyn_skirt_front_top_l_02\"),Transform=(Rotation=(X=0,Y=0,Z=0,W=1),Translation=(X=-5,Y=0,Z=0),Scale3D=(X=1,Y=1,Z=1))),(Item=(Type=Bone,Name=\"dyn_skirt_side_top_l_02\"),Transform=(Rotation=(X=0,Y=0,Z=0,W=1),Translation=(X=-5,Y=0,Z=0),Sc ...(truncated)"
    Weight=1
    WeightMinimum=0
   )
   &RigUnit_ModifyTransforms(
    ExecutePin=()
    ItemToModify="((Item=(Type=Bone,Name=\"base_skirt_back\"),Transform=(Rotation=(X=-0.2797,Y=0.1285,Z=-0.1086,W=0.9452),Translation=(X=0,Y=0,Z=0),Scale3D=(X=1,Y=1,Z=1))),(Item=(Type=Bone,Name=\"base_skirt_back_l\"),Transform=(Rotation=(X=-0.1335,Y=0.3533,Z=0.1874,W=0.9068),Tr ...(truncated)"
    Weight=1
    WeightMinimum=0
   )
   &RigUnit_ModifyTransforms(
    ExecutePin=()
    ItemToModify="((Item=(Type=Bone,Name=\"base_skirt_back\"),Transform=(Rotation=(X=-0.0328,Y=-0.0386,Z=0.6867,W=-0.7251),Translation=(X=0,Y=-1,Z=0),Scale3D=(X=1,Y=1,Z=1))),(Item=(Type=Bone,Name=\"base_skirt_back_l\"),Transform=(Rotation=(X=0.0174,Y=0.033,Z=0.6356,W=-0.7711),T ...(truncated)"
    Weight=1
    WeightMinimum=0
   )
   &RigUnit_ModifyTransforms(
    ExecutePin=()
    ItemToModify="((Item=(Type=Bone,Name=\"base_skirt_side_l\"),Transform=(Rotation=(X=-0.0394,Y=0.1819,Z=0.1268,W=0.9743),Translation=(X=0,Y=0,Z=0),Scale3D=(X=1,Y=1,Z=1))),(Item=(Type=Bone,Name=\"dyn_skirt_side_top_l_02\"),Transform=(Rotation=(X=-0.0014,Y=0.0097,Z=0.1412,W=0.9 ...(truncated)"
    Weight=1
    WeightMinimum=0
   )
   &RigUnit_ModifyTransforms(
    ExecutePin=()
    ItemToModify="((Item=(Type=Bone,Name=\"base_skirt_front_l\"),Transform=(Rotation=(X=0.0014,Y=0.0239,Z=-0.0265,W=0.9994),Translation=(X=-10,Y=4.2773,Z=3.0258),Scale3D=(X=1,Y=1,Z=1))),(Item=(Type=Bone,Name=\"base_skirt_side_l\"),Transform=(Rotation=(X=0,Y=0,Z=0,W=1),Translati ...(truncated)"
    Weight=1
    WeightMinimum=0
   )
   &RigVMFunction_MathDoubleRemap(SourceMaximum=43.5, SourceMinimum=89, TargetMaximum=1, TargetMinimum=0, Value=0)
   RigVMFunction_MathQuaternionSwingTwist = &RigVMFunction_MathQuaternionSwingTwist(
    TwistAxis=(X=0,Y=0,Z=1)
    Input=&RigUnit_GetRelativeTransformForItem(Child=(Type=Bone,Name="spine_01"), Parent=(Type=Bone,Name="pelvis"), bChildInitial=false, bParentInitial=false).RelativeTransform.Rotation
   )
   Multiply = &RigVMFunction_MathDoubleMul(
    B=&RigVMFunction_MathDoubleRemap(
     SourceMaximum=-40
     SourceMinimum=0
     TargetMaximum=1
     TargetMinimum=0
     Value=&RigVMFunction_MathQuaternionToEuler(RotationOrder=ZYX, Value=@RigVMFunction_MathQuaternionSwingTwist.Swing).Result.X
    ).Result
    A=@n34.Value
   )
   RigVMFunction_AnimEasing = &RigVMFunction_AnimEasing(SourceMaximum=1, SourceMinimum=0, TargetMaximum=1, TargetMinimum=0, Type=ExponentialEaseOut, Value=@Multiply.Result)
  ]}
 New Function_ContainedGraph {nodes [
   RigVMFunctionEntryNode()
   > &RigUnit_ParentConstraint(
    AdvancedSettings=(RotationOrderForFilter=XZY)
    Child=(Type=Bone,Name="base_skirt_front_l")
    Parents="((Item=(Type=Bone,Name=\"thigh_twist_01_r\"),Weight=0.007),(Item=(Type=Bone,Name=\"thigh_l\"),Weight=0.0088),(Item=(Type=Bone,Name=\"thigh_twist_01_l\"),Weight=0.2753),(Item=(Type=Bone,Name=\"deform_groin_l\"),Weight=0.2805),(Item=(Type=Bone,Name=\"pelvis\"),Weight= ...(truncated)"
    Weight=1
   )
   > &RigUnit_ParentConstraint(
    AdvancedSettings=(RotationOrderForFilter=XZY)
    Child=(Type=Bone,Name="base_skirt_side_l")
    Parents="((Item=(Type=Bone,Name=\"thigh_l\"),Weight=0.0256),(Item=(Type=Bone,Name=\"deform_glute_l\"),Weight=0.0322),(Item=(Type=Bone,Name=\"deform_groin_l\"),Weight=0.0574),(Item=(Type=Bone,Name=\"pelvis\"),Weight=0.399),(Item=(Type=Bone,Name=\"thigh_twist_01_l\"),Weight=0. ...(truncated)"
    Weight=1
   )
   > &RigUnit_ParentConstraint(
    AdvancedSettings=(RotationOrderForFilter=XZY)
    Child=(Type=Bone,Name="base_skirt_back_l")
    Parents=((Item=(Type=Bone,Name="thigh_twist_01_l"),Weight=0.1503),(Item=(Type=Bone,Name="deform_glute_l"),Weight=0.3559),(Item=(Type=Bone,Name="pelvis"),Weight=0.4938))
    Weight=1
   )
   > &RigUnit_ParentConstraint(
    AdvancedSettings=(RotationOrderForFilter=XZY)
    Child=(Type=Bone,Name="base_skirt_front")
    Parents="((Item=(Type=Bone,Name=\"spine_01\"),Weight=0.0021),(Item=(Type=Bone,Name=\"deform_groin_r\"),Weight=0.0541),(Item=(Type=Bone,Name=\"deform_groin_l\"),Weight=0.0701),(Item=(Type=Bone,Name=\"thigh_twist_01_r\"),Weight=0.1192),(Item=(Type=Bone,Name=\"thigh_twist_01_l ...(truncated)"
    Weight=1
   )
   > &RigUnit_ParentConstraint(
    AdvancedSettings=(RotationOrderForFilter=XZY)
    Child=(Type=Bone,Name="base_skirt_back")
    Parents="((Item=(Type=Bone,Name=\"thigh_twist_01_l\"),Weight=0.0105),(Item=(Type=Bone,Name=\"thigh_twist_01_r\"),Weight=0.0399),(Item=(Type=Bone,Name=\"deform_glute_l\"),Weight=0.0777),(Item=(Type=Bone,Name=\"deform_glute_r\"),Weight=0.1262),(Item=(Type=Bone,Name=\"pelvis\") ...(truncated)"
    Weight=1
   )
   > &RigUnit_ParentConstraint(
    AdvancedSettings=(RotationOrderForFilter=XZY)
    Child=(Type=Bone,Name="base_skirt_front_r")
    Parents="((Item=(Type=Bone,Name=\"thigh_twist_01_l\"),Weight=0.0079),(Item=(Type=Bone,Name=\"thigh_r\"),Weight=0.0082),(Item=(Type=Bone,Name=\"thigh_twist_01_r\"),Weight=0.2696),(Item=(Type=Bone,Name=\"deform_groin_r\"),Weight=0.2771),(Item=(Type=Bone,Name=\"pelvis\"),Weight ...(truncated)"
    Weight=1
   )
   > &RigUnit_ParentConstraint(
    AdvancedSettings=(RotationOrderForFilter=XZY)
    Child=(Type=Bone,Name="base_skirt_side_r")
    Parents="((Item=(Type=Bone,Name=\"thigh_r\"),Weight=0.0285),(Item=(Type=Bone,Name=\"deform_glute_r\"),Weight=0.032),(Item=(Type=Bone,Name=\"deform_groin_r\"),Weight=0.0571),(Item=(Type=Bone,Name=\"pelvis\"),Weight=0.388),(Item=(Type=Bone,Name=\"thigh_twist_01_r\"),Weight=0.4 ...(truncated)"
    Weight=1
   )
   > &RigUnit_ParentConstraint(
    AdvancedSettings=(RotationOrderForFilter=XZY)
    Child=(Type=Bone,Name="base_skirt_back_r")
    Parents=((Item=(Type=Bone,Name="thigh_twist_01_r"),Weight=0.1499),(Item=(Type=Bone,Name="deform_glute_r"),Weight=0.3585),(Item=(Type=Bone,Name="pelvis"),Weight=0.4916))
    Weight=1
   )
   > RigVMFunctionReturnNode()
  ]}
 New Function_ContainedGraph {nodes [
   RigVMFunctionEntryNode()
   > RigVMFunction_Sequence = RigVMAggregateNode()
   n1 = &RigUnit_SphericalPoseReader(
    ActiveRegionScaleFactors=(PositiveWidth=1,NegativeWidth=1,PositiveHeight=1,NegativeHeight=1)
    ActiveRegionSize=0.05
    Debug=(bDrawDebug=false,bDraw2D=False,bDrawLocalAxes=False,DebugScale=25,DebugSegments=20,DebugThickness=0.25)
    DriverAxis=(X=1,Y=0,Z=0)
    DriverItem=(Type=Bone,Name="thigh_r")
    FalloffRegionScaleFactors=(PositiveWidth=1,NegativeWidth=1,PositiveHeight=0.9,NegativeHeight=0.9)
    FalloffSize=0.65
    FlipHeightScaling=false
    FlipWidthScaling=false
    OptionalParentItem=(Type=Bone,Name="None")
    RotationOffset=(X=-90,Y=0,Z=-30)
    ExecutePin=@RigVMFunction_Sequence.A
   )
   > &RigUnit_ModifyTransforms(
    ItemToModify="((Item=(Type=Bone,Name=\"base_skirt_front_r\"),Transform=(Rotation=(X=0,Y=0,Z=-0.887,W=0.4617),Translation=(X=0,Y=0,Z=0),Scale3D=(X=1,Y=1,Z=1))),(Item=(Type=Bone,Name=\"base_skirt_side_r\"),Transform=(Rotation=(X=0.0892,Y=-0.6772,Z=-0.0892,W=0.7249),Translatio ...(truncated)"
    WeightMinimum=0
    Weight=@n1.OutputParam
   )
   > n2 = &RigUnit_SphericalPoseReader(
    ActiveRegionScaleFactors=(PositiveWidth=1,NegativeWidth=1,PositiveHeight=1,NegativeHeight=1.3)
    ActiveRegionSize=0.05
    Debug=(bDrawDebug=false,bDraw2D=False,bDrawLocalAxes=False,DebugScale=25,DebugSegments=20,DebugThickness=0.25)
    DriverAxis=(X=1,Y=0,Z=0)
    DriverItem=(Type=Bone,Name="thigh_r")
    FalloffRegionScaleFactors=(PositiveWidth=1,NegativeWidth=1,PositiveHeight=1,NegativeHeight=1.3)
    FalloffSize=0.2
    FlipHeightScaling=true
    FlipWidthScaling=true
    OptionalParentItem=(Type=Bone,Name="None")
    RotationOffset=(X=-40,Y=30,Z=0)
   )
   > &RigUnit_ModifyTransforms(
    ItemToModify="((Item=(Type=Bone,Name=\"base_skirt_front_r\"),Transform=(Rotation=(X=0,Y=0,Z=-0.342,W=0.9397),Translation=(X=0,Y=0,Z=0),Scale3D=(X=1,Y=1,Z=1))),(Item=(Type=Bone,Name=\"base_skirt_side_r\"),Transform=(Rotation=(X=0,Y=0,Z=0,W=1),Translation=(X=0,Y=0,Z=0),Scale3 ...(truncated)"
    WeightMinimum=0
    Weight=@n2.OutputParam
   )
   > n3 = &RigUnit_SphericalPoseReader(
    ActiveRegionScaleFactors=(PositiveWidth=1,NegativeWidth=1,PositiveHeight=1,NegativeHeight=1)
    ActiveRegionSize=0.05
    Debug=(bDrawDebug=false,bDraw2D=False,bDrawLocalAxes=False,DebugScale=25,DebugSegments=20,DebugThickness=0.25)
    DriverAxis=(X=1,Y=0,Z=0)
    DriverItem=(Type=Bone,Name="thigh_r")
    FalloffRegionScaleFactors=(PositiveWidth=1,NegativeWidth=1,PositiveHeight=0.9,NegativeHeight=0.9)
    FalloffSize=0.65
    FlipHeightScaling=true
    FlipWidthScaling=true
    OptionalParentItem=(Type=Bone,Name="None")
    RotationOffset=(X=0,Y=27,Z=0)
   )
   > &RigUnit_ModifyTransforms(
    ItemToModify="((Item=(Type=Bone,Name=\"offset_skirt_side_r\"),Transform=(Rotation=(X=0,Y=0,Z=-0.766,W=0.6428),Translation=(X=0,Y=0,Z=0),Scale3D=(X=1,Y=1,Z=1))),(Item=(Type=Bone,Name=\"offset_skirt_front_r\"),Transform=(Rotation=(X=-0.1209,Y=0.6127,Z=-0.0602,W=0.7787),Transl ...(truncated)"
    WeightMinimum=0
    Weight=@n3.OutputParam
   )
   > n4 = &RigUnit_SphericalPoseReader(
    ActiveRegionScaleFactors=(PositiveWidth=1,NegativeWidth=1,PositiveHeight=1,NegativeHeight=1)
    ActiveRegionSize=0.05
    Debug=(bDrawDebug=false,bDraw2D=False,bDrawLocalAxes=False,DebugScale=25,DebugSegments=20,DebugThickness=0.25)
    DriverAxis=(X=1,Y=0,Z=0)
    DriverItem=(Type=Bone,Name="thigh_r")
    FalloffRegionScaleFactors=(PositiveWidth=1,NegativeWidth=1,PositiveHeight=1,NegativeHeight=1)
    FalloffSize=0.55
    FlipHeightScaling=False
    FlipWidthScaling=False
    OptionalParentItem=(Type=Bone,Name="None")
    RotationOffset=(X=90,Y=0,Z=20)
   )
   > &RigUnit_ModifyTransforms(
    ItemToModify="((Item=(Type=Bone,Name=\"base_skirt_back_r\"),Transform=(Rotation=(X=0.183,Y=0.183,Z=-0.683,W=0.683),Translation=(X=0,Y=0,Z=0),Scale3D=(X=1,Y=1,Z=1))),(Item=(Type=Bone,Name=\"base_skirt_side_r\"),Transform=(Rotation=(X=-0.0334,Y=0.3812,Z=0.0805,W=0.9204),Trans ...(truncated)"
    WeightMinimum=0
    Weight=@n4.OutputParam
   )
   n5 = &RigUnit_SphericalPoseReader(
    ActiveRegionScaleFactors=(PositiveWidth=1,NegativeWidth=1,PositiveHeight=1,NegativeHeight=1)
    ActiveRegionSize=0.05
    Debug=(bDrawDebug=false,bDraw2D=False,bDrawLocalAxes=False,DebugScale=25,DebugSegments=20,DebugThickness=0.25)
    DriverAxis=(X=-1,Y=0,Z=0)
    DriverItem=(Type=Bone,Name="thigh_l")
    FalloffRegionScaleFactors=(PositiveWidth=1,NegativeWidth=1,PositiveHeight=0.9,NegativeHeight=0.9)
    FalloffSize=0.65
    FlipHeightScaling=False
    FlipWidthScaling=False
    OptionalParentItem=(Type=Bone,Name="None")
    RotationOffset=(X=90,Y=0,Z=-30)
    ExecutePin=@RigVMFunction_Sequence.B
   )
   > &RigUnit_ModifyTransforms(
    ItemToModify="((Item=(Type=Bone,Name=\"base_skirt_front_l\"),Transform=(Rotation=(X=0,Y=0,Z=-0.8829,W=0.4695),Translation=(X=0,Y=0,Z=0),Scale3D=(X=1,Y=1,Z=1))),(Item=(Type=Bone,Name=\"base_skirt_side_l\"),Transform=(Rotation=(X=0.0892,Y=-0.6772,Z=-0.0892,W=0.7249),Translati ...(truncated)"
    WeightMinimum=0
    Weight=@n5.OutputParam
   )
   > n6 = &RigUnit_SphericalPoseReader(
    ActiveRegionScaleFactors=(PositiveWidth=1,NegativeWidth=1,PositiveHeight=1.3,NegativeHeight=1)
    ActiveRegionSize=0.05
    Debug=(bDrawDebug=false,bDraw2D=False,bDrawLocalAxes=False,DebugScale=25,DebugSegments=20,DebugThickness=0.25)
    DriverAxis=(X=-1,Y=0,Z=0)
    DriverItem=(Type=Bone,Name="thigh_l")
    FalloffRegionScaleFactors=(PositiveWidth=1,NegativeWidth=1,PositiveHeight=1.3,NegativeHeight=1)
    FalloffSize=0.2
    FlipHeightScaling=False
    FlipWidthScaling=False
    OptionalParentItem=(Type=Bone,Name="None")
    RotationOffset=(X=140,Y=30,Z=0)
   )
   > &RigUnit_ModifyTransforms(
    ItemToModify="((Item=(Type=Bone,Name=\"base_skirt_front_l\"),Transform=(Rotation=(X=0,Y=0,Z=-0.342,W=0.9397),Translation=(X=0,Y=0,Z=0),Scale3D=(X=1,Y=1,Z=1))),(Item=(Type=Bone,Name=\"base_skirt_side_l\"),Transform=(Rotation=(X=0,Y=0,Z=0,W=1),Translation=(X=0,Y=0,Z=0),Scale3 ...(truncated)"
    WeightMinimum=0
    Weight=@n6.OutputParam
   )
   > n7 = &RigUnit_SphericalPoseReader(
    ActiveRegionScaleFactors=(PositiveWidth=1,NegativeWidth=1,PositiveHeight=1,NegativeHeight=1)
    ActiveRegionSize=0.05
    Debug=(bDrawDebug=false,bDraw2D=False,bDrawLocalAxes=False,DebugScale=25,DebugSegments=20,DebugThickness=0.25)
    DriverAxis=(X=-1,Y=0,Z=0)
    DriverItem=(Type=Bone,Name="thigh_l")
    FalloffRegionScaleFactors=(PositiveWidth=1,NegativeWidth=1,PositiveHeight=0.9,NegativeHeight=0.9)
    FalloffSize=0.65
    FlipHeightScaling=False
    FlipWidthScaling=False
    OptionalParentItem=(Type=Bone,Name="None")
    RotationOffset=(X=180,Y=27,Z=0)
   )
   > &RigUnit_ModifyTransforms(
    ItemToModify="((Item=(Type=Bone,Name=\"offset_skirt_side_l\"),Transform=(Rotation=(X=0,Y=0,Z=-0.766,W=0.6428),Translation=(X=0,Y=0,Z=0),Scale3D=(X=1,Y=1,Z=1))),(Item=(Type=Bone,Name=\"offset_skirt_front_l\"),Transform=(Rotation=(X=-0.0941,Y=0.6174,Z=-0.0941,W=0.7753),Transl ...(truncated)"
    WeightMinimum=0
    Weight=@n7.OutputParam
   )
   > n8 = &RigUnit_SphericalPoseReader(
    ActiveRegionScaleFactors=(PositiveWidth=1,NegativeWidth=1,PositiveHeight=1,NegativeHeight=1)
    ActiveRegionSize=0.05
    Debug=(bDrawDebug=false,bDraw2D=False,bDrawLocalAxes=False,DebugScale=25,DebugSegments=20,DebugThickness=0.25)
    DriverAxis=(X=-1,Y=0,Z=0)
    DriverItem=(Type=Bone,Name="thigh_l")
    FalloffRegionScaleFactors=(PositiveWidth=1,NegativeWidth=1,PositiveHeight=1,NegativeHeight=1)
    FalloffSize=0.55
    FlipHeightScaling=False
    FlipWidthScaling=False
    OptionalParentItem=(Type=Bone,Name="None")
    RotationOffset=(X=-90,Y=0,Z=20)
   )
   > &RigUnit_ModifyTransforms(
    ItemToModify="((Item=(Type=Bone,Name=\"base_skirt_back_l\"),Transform=(Rotation=(X=0.183,Y=0.183,Z=-0.683,W=0.683),Translation=(X=0,Y=0,Z=0),Scale3D=(X=1,Y=1,Z=1))),(Item=(Type=Bone,Name=\"base_skirt_side_l\"),Transform=(Rotation=(X=-0.0334,Y=0.3812,Z=0.0805,W=0.9204),Trans ...(truncated)"
    WeightMinimum=0
    Weight=@n8.OutputParam
   )
   RigVMFunctionReturnNode(ExecuteContext=@RigVMFunction_Sequence.D)
   &RigUnit_ModifyTransforms(
    ItemToModify=((Item=(Type=Bone,Name="base_skirt_front"),Transform=(Rotation=(X=0,Y=0,Z=-0.866,W=0.5),Translation=(X=0,Y=0,Z=0),Scale3D=(X=1,Y=1,Z=1))))
    WeightMinimum=0
    Weight=&RigVMFunction_MathDoubleMax(B=@n5.OutputParam, A=@n1.OutputParam).Result
    ExecutePin=@RigVMFunction_Sequence.C
   )
   > &RigUnit_ModifyTransforms(
    ItemToModify=((Item=(Type=Bone,Name="base_skirt_back"),Transform=(Rotation=(X=0,Y=0,Z=-0.6428,W=0.766),Translation=(X=0,Y=0,Z=0),Scale3D=(X=1,Y=1,Z=1))))
    WeightMinimum=0
    Weight=&RigVMFunction_MathDoubleMax(B=@n8.OutputParam, A=@n4.OutputParam).Result
   )
  ]}
 RigVMFunctionLibrary {nodes [
   SkirtSetup = RigVMCollapseNode(ExecuteContext=())
   skirtBackup = RigVMCollapseNode(ExecuteContext=())
   skirt_constraints = RigVMCollapseNode(ExecuteContext=())
   skirt_setup = RigVMCollapseNode(ExecuteContext=())
  ]}
 RigVMFunction_Sequence_1_ContainedGraph {nodes [
   n1 = RigVMFunctionEntryNode()
   RigVMFunction_Sequence = &RigVMFunction_Sequence(ExecuteContext=@n1.ExecuteContext)
   n2 = &RigVMFunction_Sequence(ExecuteContext=@RigVMFunction_Sequence.B)
   n3 = &RigVMFunction_Sequence(ExecuteContext=@n2.B)
   n4 = &RigVMFunction_Sequence(ExecuteContext=@n3.B)
   n5 = &RigVMFunction_Sequence(ExecuteContext=@n4.B)
   n6 = &RigVMFunction_Sequence(ExecuteContext=@n5.B)
   n7 = &RigVMFunction_Sequence(ExecuteContext=@n6.B)
   RigVMFunctionReturnNode(A=@RigVMFunction_Sequence.A, B=@n2.A, C=@n3.A, D=@n4.A, E=@n5.A, F=@n6.A, G=@n7.A, H=@n7.B)
  ]}
 RigVMFunction_Sequence_1_ContainedGraph {nodes [
   n1 = RigVMFunctionEntryNode()
   RigVMFunction_Sequence = &RigVMFunction_Sequence(ExecuteContext=@n1.ExecuteContext)
   n2 = &RigVMFunction_Sequence(ExecuteContext=@RigVMFunction_Sequence.B)
   n3 = &RigVMFunction_Sequence(ExecuteContext=@n2.B)
   n4 = &RigVMFunction_Sequence(ExecuteContext=@n3.B)
   n5 = &RigVMFunction_Sequence(ExecuteContext=@n4.B)
   n6 = &RigVMFunction_Sequence(ExecuteContext=@n5.B)
   n7 = &RigVMFunction_Sequence(ExecuteContext=@n6.B)
   n8 = &RigVMFunction_Sequence(ExecuteContext=@n7.B)
   n9 = &RigVMFunction_Sequence(ExecuteContext=@n8.B)
   n10 = &RigVMFunction_Sequence(ExecuteContext=@n9.B)
   RigVMFunctionReturnNode(A=@RigVMFunction_Sequence.A, B=@n2.A, C=@n3.A, D=@n4.A, E=@n5.A, F=@n6.A, G=@n7.A, H=@n8.A, I=@n9.A, J=@n10.A, K=@n10.B)
  ]}
 RigVMFunction_Sequence_1_ContainedGraph {nodes [
   n1 = RigVMFunctionEntryNode()
   RigVMFunction_Sequence = &RigVMFunction_Sequence(ExecuteContext=@n1.ExecuteContext)
   n2 = &RigVMFunction_Sequence(ExecuteContext=@RigVMFunction_Sequence.B)
   n3 = &RigVMFunction_Sequence(ExecuteContext=@n2.B)
   RigVMFunctionReturnNode(A=@RigVMFunction_Sequence.A, B=@n2.A, C=@n3.A, D=@n3.B)
  ]}
 RigVMModel {nodes [
   n1 = &RigUnit_PrepareForExecution()
   &RigUnit_HierarchyImportFromSkeleton(NameSpace=None, bIncludeCurves=False, bIncludeMeshSockets=False, bIncludeVirtualBones=True, ExecutePin=@n1.ExecutePin)
   > /Script/ControlRig.RigUnit_HierarchyAddBone(
    Name=thigh_l_ref
    Parent=(Type=Bone,Name="pelvis")
    Space=LocalSpace
    Transform=(Rotation=(X=-0.0009,Y=-0.0355,Z=-0.0155,W=0.9992),Translation=(X=-1.0855,Y=-0.2109,Z=-8.7053),Scale3D=(X=1,Y=1,Z=1))
   ) {is=struct RigVMUnitNode}
   > &RigUnit_HierarchyAddBone(
    Name=thigh_r_ref
    Parent=(Type=Bone,Name="pelvis")
    Space=LocalSpace
    Transform=(Rotation=(X=0.0355,Y=-0.0009,Z=0.9992,W=0.0155),Translation=(X=-1.0855,Y=-0.2109,Z=8.7053),Scale3D=(X=1,Y=1,Z=1))
   )
   RigUnit_BeginExecution = &RigUnit_BeginExecution()
   FN_DeformRig_FNC = RigVMFunctionReferenceNode(
    DeformRig_ Curve="(deformRig_l_upperarm_up_alpha_72_#1=False,deformRig_r_upperarm_up_alpha_73_#2=False,deformRig_l_upperarm_fr_alpha_74_#3=False,deformRig_r_upperarm_fr_alpha_75_CB51B8 ...(truncated)"
    DeformRig_Bool="(upperarm_1_#4=True,elbow_3_#5=True,wrist_7_#6=True,groin_8_#7=True,glute_17_#8=True,knee_9_40C408324CA261 ...(truncated)"
    DeformRig_Debug=(upperarm_debug_18_#4=False,wrist_debug_20_#6=False,lat_debug_26_#9=False,groinGlute_debug_27_#7=False)
    DeformRig_Value="(upperarm_up_18_#10=(Rotation=(X=0,Y=0,Z=0,W=1),Translation=(X=0,Y=0,Z=8),Scale3D=(X=0.7,Y=1,Z=1)),upperarm_fwd_fr_19_#11=(Rotation=(X=0,Y=0,Z=0,W=1),Translation=(X=0,Y=10,Z=0),Scale3D=(X=0.5,Y=1,Z= ...(truncated)"
    ExecuteContext=@RigUnit_BeginExecution.ExecutePin
   )
   > skirt_setup = RigVMFunctionReferenceNode()
   SkirtSetup = RigVMFunctionReferenceNode(ExecuteContext=())
   addFKCtrl = RigVMFunctionReferenceNode(
    Custom_FKCtrl=()
    ExecuteContext=()
    Options_FKCtrl="(pelvis_1_#12=false,spine_3_#13=true,clavicle_5_#14=false,arm_7_#15=false,leg_9_#16=true,head_11_46DC15EF465CD2 ...(truncated)"
    Shape_Settings="(pelvis_2_#17=(bVisible=True,Name=\"Hexagon_Thin\",Color=(R=0.0117,G=0.8281,B=1,A=1),Transform=(Rotation=(X=0,Y=-0.7071,Z=0,W=0.7071),Translation=(X=2,Y=0,Z=0),Scale3D=(X=5,Y=5,Z=7))),spine_4_#18=(bVi ...(truncated)"
   )
   setFKCtrl_1 = RigVMFunctionReferenceNode(ExecuteContext=(), setFK_Ctrl=false)
   skirt_constraints = RigVMFunctionReferenceNode(ExecuteContext=())
  ]}
]
variables [
 SphericalPoseReader_MidFront_L {type=float, default=0}
 SphericalPoseReader_MidFront_R {type=float, default=0}
 isBaseBallSliding {type=double, default=0}
 isCrouching {type=double, default=0}
 isMoving2D {type=double, default=0}
 isRailGringind {type=double, default=0}
 isSkyDiving {type=double, default=0}
 isSliding {type=double, default=0}
 isSprinting {type=double, default=0}
]
