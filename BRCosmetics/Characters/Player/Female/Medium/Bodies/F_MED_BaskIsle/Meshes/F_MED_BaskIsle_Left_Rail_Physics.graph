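# Physics asset for the BaskIsle medium female player body, simulating the left half of its skirt plus the front and back center strips.
# Five capsule chains (front, front-left, side-left, back-left, back) each link a small offset anchor to three skirt segments.
# Constraints lock twist and allow small single-axis swing, stiffer at the front and back roots.
# Unconstrained thigh, calf and foot capsules on both legs provide bodies for skirt collision.
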PhysicsAsset /BRCosmetics/Characters/Player/Female/Medium/Bodies/F_MED_BaskIsle/Meshes/F_MED_BaskIsle_Left_Rail_Physics
bodies {
 calf_l [{
   shapeType=Capsule
   length=27.1033
   radius=6.9673
   shapeName=thigh_r_capsule
   center=<-20.372 -2.0988 -1.1179>
   rotation=<rot 0.228 -90.0044 -91.8958>
  }]
 calf_r [{
   shapeType=Capsule
   length=27.1033
   radius=6.9673
   shapeName=thigh_r_capsule
   center=<20.372 2.0988 1.1179>
   rotation=<rot -0.228 -90.0044 -91.8958>
  }]
 dyn_skirt_back_01 [{shapeType=Capsule, length=16, radius=5, shapeName=dyn_skirt_back_02_capsule, center=<2 -2 0>}]
 dyn_skirt_back_02 [{shapeType=Capsule, length=17, radius=5, shapeName=dyn_skirt_back_02_capsule, center=<2 -2 0>}]
 dyn_skirt_back_l_01 [{shapeType=Capsule, length=16, radius=5, shapeName=dyn_skirt_back_l_02_capsule, center=<2 -2.7 0>, rotation=<rot 0 0 -3>}]
 dyn_skirt_back_l_02 [{shapeType=Capsule, length=17, radius=5, shapeName=dyn_skirt_back_l_02_capsule, center=<2 -2.7 0>, rotation=<rot 0 0 -3>}]
 dyn_skirt_back_top_02 [{shapeType=Capsule, length=16, radius=5.5, shapeName=dyn_skirt_back_top_02_capsule, center=<2 -2 0>}]
 dyn_skirt_back_top_l_02 [{shapeType=Capsule, length=16, radius=5.5, shapeName=dyn_skirt_back_top_l_02_capsule, center=<2 -3 0>, rotation=<rot 0 0 -3>}]
 dyn_skirt_front_01 [{shapeType=Capsule, length=16, radius=5, shapeName=dyn_skirt_front_02_capsule, center=<2 -2 0>}]
 dyn_skirt_front_02 [{shapeType=Capsule, length=17, radius=5, shapeName=dyn_skirt_front_02_capsule, center=<2 -2 0>}]
 dyn_skirt_front_l_01 [{shapeType=Capsule, length=16, radius=5, shapeName=dyn_skirt_front_l_02_capsule, center=<2 -3 -2.1177>, rotation=<rot 0 0 -7.8671>}]
 dyn_skirt_front_l_02 [{shapeType=Capsule, length=17, radius=5, shapeName=dyn_skirt_front_l_02_capsule, center=<2 -3 -2.0041>, rotation=<rot 0 0 -7.8671>}]
 dyn_skirt_front_top_02 [{shapeType=Capsule, length=16, radius=5, shapeName=dyn_skirt_front_top_02_capsule, center=<2 -2 0>}]
 dyn_skirt_front_top_l_02 [{shapeType=Capsule, length=16, radius=5, shapeName=dyn_skirt_front_top_l_02_capsule, center=<2 -3 -2.0041>, rotation=<rot 0 0 -7.8671>}]
 dyn_skirt_side_l_01 [{shapeType=Capsule, length=16, radius=5, shapeName=dyn_skirt_side_l_02_capsule, center=<2 -4 0>}]
 dyn_skirt_side_l_02 [{shapeType=Capsule, length=17, radius=5, shapeName=dyn_skirt_side_l_02_capsule, center=<2 -4 0>}]
 dyn_skirt_side_top_l_02 [{shapeType=Capsule, length=16, radius=5, shapeName=dyn_skirt_side_top_l_02_capsule, center=<2 -4 0>}]
 foot_l [{
   shapeType=Capsule
   length=13.2015
   radius=8.5072
   shapeName=foot_r_capsule
   center=<-4.894 7.2004 0.0534>
   rotation=<rot 0.5093 12.911 -85.9431>
  }]
 foot_r [{
   shapeType=Capsule
   length=13.2015
   radius=8.5072
   shapeName=foot_r_capsule
   center=<4.894 -7.2004 -0.0534>
   rotation=<rot 0.5093 12.911 -85.9431>
  }]
 offset_skirt_back [{shapeType=Capsule, length=1, radius=1, shapeName=offset_skirt_back_capsule}]
 offset_skirt_back_l [{shapeType=Capsule, length=1, radius=1, shapeName=offset_skirt_back_l_capsule}]
 offset_skirt_front [{shapeType=Capsule, length=1, radius=1, shapeName=offset_skirt_front_capsule}]
 offset_skirt_front_l [{shapeType=Capsule, length=1, radius=1, shapeName=offset_skirt_front_l_capsule}]
 offset_skirt_side_l [{shapeType=Capsule, length=1, radius=1, shapeName=offset_skirt_side_l_capsule}]
 thigh_l [{
   shapeType=Capsule
   length=33.0434
   radius=8
   shapeName=thigh_r_capsule
   center=<-20.8954 0.3735 0.1107>
   rotation=<rot 0.2451 -89.4896 -91.8937>
  }]
 thigh_r [{
   shapeType=Capsule
   length=33.0434
   radius=8
   shapeName=thigh_r_capsule
   center=<20.8954 -0.3735 -0.1107>
   rotation=<rot -0.2451 -89.4896 -91.8937>
  }]
}
constraints [
 {
  bone1Name=dyn_skirt_back_02
  bone2Name=dyn_skirt_back_01
  swing1LimitDegrees=10
  swing2LimitDegrees=0
  twistLimitDegrees=1
  twistMotion=Locked
 }
 {
  bone1Name=dyn_skirt_back_l_02
  bone2Name=dyn_skirt_back_l_01
  swing1LimitDegrees=10
  swing2LimitDegrees=0
  twistLimitDegrees=1
  twistMotion=Locked
 }
 {
  bone1Name=dyn_skirt_front_02
  bone2Name=dyn_skirt_front_01
  swing1LimitDegrees=2
  swing2LimitDegrees=0
  twistLimitDegrees=1
  twistMotion=Locked
 }
 {
  bone1Name=dyn_skirt_front_l_02
  bone2Name=dyn_skirt_front_l_01
  swing1LimitDegrees=2
  swing2LimitDegrees=0
  twistLimitDegrees=1
  twistMotion=Locked
 }
 {
  bone1Name=dyn_skirt_side_l_02
  bone2Name=dyn_skirt_side_l_01
  swing1LimitDegrees=10
  swing2LimitDegrees=0
  twistLimitDegrees=1
  twistMotion=Locked
 }
 {
  bone1Name=dyn_skirt_back_01
  bone2Name=dyn_skirt_back_top_02
  swing1LimitDegrees=10
  swing2LimitDegrees=0
  twistLimitDegrees=1
  twistMotion=Locked
 }
 {
  bone1Name=dyn_skirt_back_l_01
  bone2Name=dyn_skirt_back_top_l_02
  swing1LimitDegrees=10
  swing2LimitDegrees=0
  twistLimitDegrees=1
  twistMotion=Locked
 }
 {
  bone1Name=dyn_skirt_front_01
  bone2Name=dyn_skirt_front_top_02
  swing1LimitDegrees=2
  swing2LimitDegrees=0
  twistLimitDegrees=1
  twistMotion=Locked
 }
 {
  bone1Name=dyn_skirt_front_l_01
  bone2Name=dyn_skirt_front_top_l_02
  swing1LimitDegrees=2
  swing2LimitDegrees=0
  twistLimitDegrees=1
  twistMotion=Locked
 }
 {
  bone1Name=dyn_skirt_side_l_01
  bone2Name=dyn_skirt_side_top_l_02
  swing1LimitDegrees=10
  swing2LimitDegrees=0
  twistLimitDegrees=1
  twistMotion=Locked
 }
 {
  bone1Name=dyn_skirt_back_top_02
  bone2Name=offset_skirt_back
  swing1LimitDegrees=1
  swing2LimitDegrees=0
  twistLimitDegrees=1
  twistMotion=Locked
 }
 {
  bone1Name=dyn_skirt_back_top_l_02
  bone2Name=offset_skirt_back_l
  swing1LimitDegrees=1
  swing2LimitDegrees=0
  twistLimitDegrees=1
  twistMotion=Locked
 }
 {
  bone1Name=dyn_skirt_front_top_02
  bone2Name=offset_skirt_front
  swing1LimitDegrees=2
  swing2LimitDegrees=0
  twistLimitDegrees=1
  twistMotion=Locked
 }
 {
  bone1Name=dyn_skirt_front_top_l_02
  bone2Name=offset_skirt_front_l
  swing1LimitDegrees=2
  swing2LimitDegrees=0
  twistLimitDegrees=1
  twistMotion=Locked
 }
 {
  bone1Name=dyn_skirt_side_top_l_02
  bone2Name=offset_skirt_side_l
  swing1LimitDegrees=10
  swing2LimitDegrees=0
  twistLimitDegrees=1
  twistMotion=Locked
 }
]
