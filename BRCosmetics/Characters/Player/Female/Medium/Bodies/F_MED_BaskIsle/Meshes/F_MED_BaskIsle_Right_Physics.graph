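# Physics asset for the right side of the skirt on the F_MED_BaskIsle female player cosmetic body.
# Front, side and back chains each link three capsule bodies to a small offset-bone capsule through constraints with tight swing and twist limits.
# Thigh, calf and foot capsules on both legs have no constraints, apparently serving as colliders for the skirt.
# The three chains have no constraints between one another.
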PhysicsAsset /BRCosmetics/Characters/Player/Female/Medium/Bodies/F_MED_BaskIsle/Meshes/F_MED_BaskIsle_Right_Physics
bodies {
 calf_l [{
   shapeType=Capsule
   length=27.1033
   radius=6.9673
   shapeName=thigh_r_capsule
   center=<-20.372 -2.0988 -1.1179>
   rotation=<rot 0.228 -90.0044 -91.8958>
  }]
 calf_r [{
   shapeType=Capsule
   length=27.1033
   radius=6.9673
   shapeName=thigh_r_capsule
   center=<20.372 2.0988 1.1179>
   rotation=<rot -0.228 -90.0044 -91.8958>
  }]
 dyn_skirt_back_r_01 [{shapeType=Capsule, length=16, radius=5, shapeName=dyn_skirt_back_r_02_capsule, center=<-2 2.7 0>}]
 dyn_skirt_back_r_02 [{shapeType=Capsule, length=17, radius=5, shapeName=dyn_skirt_back_r_02_capsule, center=<-2 2.7 0>}]
 dyn_skirt_back_top_r_02 [{shapeType=Capsule, length=16, radius=5, shapeName=dyn_skirt_back_top_r_02_capsule, center=<-2 3 0>}]
 dyn_skirt_front_r_01 [{shapeType=Capsule, length=16, radius=5.5, shapeName=dyn_skirt_front_l_02_capsule, center=<-2 4 2.1177>, rotation=<rot 0 0 -7.8671>}]
 dyn_skirt_front_r_02 [{shapeType=Capsule, length=17, radius=5.5, shapeName=dyn_skirt_front_l_02_capsule, center=<-2 2 2.0041>, rotation=<rot 0 0 -7.8671>}]
 dyn_skirt_front_top_r_02 [{shapeType=Capsule, length=16, radius=5.5, shapeName=dyn_skirt_front_top_l_02_capsule, center=<-2 4.5 2.2178>, rotation=<rot 0 0 -7.8671>}]
 dyn_skirt_side_r_01 [{shapeType=Capsule, length=16, radius=5, shapeName=dyn_skirt_side_r_02_capsule, center=<-2 3 0>}]
 dyn_skirt_side_r_02 [{shapeType=Capsule, length=17, radius=5, shapeName=dyn_skirt_side_r_02_capsule, center=<-2 3 0>}]
 dyn_skirt_side_top_r_02 [{shapeType=Capsule, length=16, radius=5, shapeName=dyn_skirt_side_top_r_02_capsule, center=<-2 4 0>}]
 foot_l [{
   shapeType=Capsule
   length=13.2015
   radius=8.5072
   shapeName=foot_r_capsule
   center=<-4.894 7.2004 0.0534>
   rotation=<rot 0.5093 12.911 -85.9431>
  }]
 foot_r [{
   shapeType=Capsule
   length=13.2015
   radius=8.5072
   shapeName=foot_r_capsule
   center=<4.894 -7.2004 -0.0534>
   rotation=<rot 0.5093 12.911 -85.9431>
  }]
 offset_skirt_back_r [{shapeType=Capsule, length=1, radius=1, shapeName=offset_skirt_back_r_capsule}]
 offset_skirt_front_r [{shapeType=Capsule, length=1, radius=1, shapeName=offset_skirt_front_r_capsule}]
 offset_skirt_side_r [{shapeType=Capsule, length=1, radius=1, shapeName=offset_skirt_side_r_capsule}]
 thigh_l [{
   shapeType=Capsule
   length=33.0434
   radius=8
   shapeName=thigh_r_capsule
   center=<-20.8954 0.3735 0.1107>
   rotation=<rot -0.2451 -89.4896 -91.8937>
  }]
 thigh_r [{
   shapeType=Capsule
   length=33.0434
   radius=8
   shapeName=thigh_r_capsule
   center=<20.8954 -0.3735 -0.1107>
   rotation=<rot -0.2451 -89.4896 -91.8937>
  }]
}
constraints [
 {bone1Name=dyn_skirt_back_r_02, bone2Name=dyn_skirt_back_r_01, swing1LimitDegrees=10, swing2LimitDegrees=5, twistLimitDegrees=5}
 {bone1Name=dyn_skirt_front_r_02, bone2Name=dyn_skirt_front_r_01, swing1LimitDegrees=10, swing2LimitDegrees=5, twistLimitDegrees=5}
 {bone1Name=dyn_skirt_side_r_02, bone2Name=dyn_skirt_side_r_01, swing1LimitDegrees=10, swing2LimitDegrees=5, twistLimitDegrees=5}
 {bone1Name=dyn_skirt_back_r_01, bone2Name=dyn_skirt_back_top_r_02, swing1LimitDegrees=10, swing2LimitDegrees=5, twistLimitDegrees=5}
 {bone1Name=dyn_skirt_front_r_01, bone2Name=dyn_skirt_front_top_r_02, swing1LimitDegrees=10, swing2LimitDegrees=5, twistLimitDegrees=5}
 {bone1Name=dyn_skirt_side_r_01, bone2Name=dyn_skirt_side_top_r_02, swing1LimitDegrees=10, swing2LimitDegrees=5, twistLimitDegrees=5}
 {bone1Name=dyn_skirt_back_top_r_02, bone2Name=offset_skirt_back_r, swing1LimitDegrees=10, swing2LimitDegrees=5, twistLimitDegrees=5}
 {bone1Name=dyn_skirt_front_top_r_02, bone2Name=offset_skirt_front_r, swing1LimitDegrees=10, swing2LimitDegrees=5, twistLimitDegrees=5}
 {bone1Name=dyn_skirt_side_top_r_02, bone2Name=offset_skirt_side_r, swing1LimitDegrees=10, swing2LimitDegrees=5, twistLimitDegrees=5}
]
