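# Post-process-style AnimBP for the F_MED_BaskIsle female cosmetic body, copying the attached parent's pose and adding skirt and physics motion.
# A Control Rig takes movement-state floats from the EventGraph, then feeds thigh-relative left and right RigidBody chains.
# A bool blend pairs these with a rail-physics variant before small skirt-back bone offsets reach output.
# Cape or long backbling reduces main physics alpha; the front-end idle layered blend does not reach output.
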AnimBlueprint /BRCosmetics/Characters/Player/Female/Medium/Bodies/F_MED_BaskIsle/Meshes/F_MED_BaskIsle_AnimBP
target_skeleton=&/F_MED_BaskIsle_Skeleton
graphs [
 AnimGraph {nodes [
   n1 = Layered blend per bone(BlendPoses_0=Slot 'DefaultSlot'(), BlendWeights_0=Get Custom Idle is Playing Float() -> (Custom Idle is Playing Float:double)) {layerSetup [{branchFilters [
       {boneName=base_skirt_back}
       {boneName=base_skirt_back_l}
       {boneName=base_skirt_back_r}
       {boneName=base_skirt_front}
       {boneName=base_skirt_front_l}
       {boneName=base_skirt_front_r}
       {boneName=base_skirt_side_l}
       {boneName=base_skirt_side_r}
      ]}]}
   Local To Component(LocalPose=@n1)
   n2 = "Transform (Modify) Bone - Bone: offset_skirt_back_l"!(
    ComponentPose=Local To Component(LocalPose=Use cached pose 'PostDyn Cache'())
    bAlphaBoolEnabled:bool=True
    Translation="0,-0.2,0"
    Rotation="0,0,0"
    Scale="1,1,1"
   ) -> (Pose) {
    alphaInputType=Bool
    rotationMode=BMM_Additive
    rotationSpace=BCS_BoneSpace
    translationMode=BMM_Additive
    translationSpace=BCS_BoneSpace
    alphaBoolBlend {blendOption=Linear}
    boneToModify {boneName=offset_skirt_back_l}
    translation {y=-0.2}
   }
   n3 = "Transform (Modify) Bone - Bone: offset_skirt_back_r"!(ComponentPose=@n2, bAlphaBoolEnabled:bool=True, Translation="0,0.2,0", Rotation="0,0,0", Scale="1,1,1") -> (Pose) {
    alphaInputType=Bool
    rotationMode=BMM_Additive
    rotationSpace=BCS_BoneSpace
    translationMode=BMM_Additive
    translationSpace=BCS_BoneSpace
    alphaBoolBlend {blendOption=Linear}
    boneToModify {boneName=offset_skirt_back_r}
    translation {y=0.2}
   }
   Output Pose(Result=Component To Local(ComponentPose=@n3)) {layerGroup=None}
   n4 = Control Rig!(
    Source=Copy Pose From Mesh() {bUseAttachedParent=true}
    Alpha
    isCrouching:double=Get isCrouching() -> (isCrouching:double)
    isSliding:double=Get isSliding() -> (isSliding:double)
    isBaseBallSliding:double=Get isBaseBallSliding() -> (isBaseBallSliding:double)
    isSprinting:double=Get isSprinting() -> (isSprinting:double)
    isSkyDiving:double=Get isSkyDiving() -> (isSkyDiving:double)
    isRailGringind:double=Get isRailGrinding() -> (isRailGrinding:double)
    isMoving2D:double=Get isMoving2D() -> (isMoving2D:double)
   ) -> (Pose) {
    alphaBoolBlend {blendOption=Linear}
    controlRigAssetReference {blueprintRigClass=&/F_MED_BaskIsle_CtrlRig.F_MED_BaskIsle_CtrlRig_C}
   }
   Save cached pose 'CR_CtrlRig'(Pose=@n4) {cacheName=CR_CtrlRig}
   n5 = Get RB_Value() -> (RB_Value:double)
   n6 = RigidBody!(
    ComponentPose=Local To Component(LocalPose=Use cached pose 'CR_CtrlRig'())
    Alpha=@n5
    OverrideWorldGravity:Vector="0,0,-980"
    ExternalForce
    SimSpaceSettings:Sim Space Settings Structure=(WorldAlpha=1,VelocityScaleZ=0,DampingAlpha=1,MaxLinearVelocity=500,MaxAngularVelocity=3,MaxLinearAcceleration=700,MaxAngularAcceleration=1000,ExternalLinearDragV=(X=0.1,Y=0.1,Z=0),ExternalLinearVelocity=(X=0,Y=0,Z=0),ExternalAngularVelocity=(X=0,Y=0,Z=0))
   ) -> (Pose) {
    bClampLinearTranslationLimitToRefPose=true
    bOverrideWorldGravity=true
    overridePhysicsAsset=&/F_MED_BaskIsle_Right_Physics
    simulationSpace=BaseBoneSpace
    alphaBoolBlend {blendOption=Linear}
    baseBoneRef {boneName=thigh_r}
    componentAppliedLinearAccClamp=<0 0 0>
    overrideWorldGravity {z=-980}
    simSpaceSettings {
     maxAngularAcceleration=1000
     maxAngularVelocity=3
     maxLinearAcceleration=700
     maxLinearVelocity=500
     velocityScaleZ=0
     worldAlpha=1
     externalLinearDragV=<0.1 0.1>
    }
   }
   n7 = RigidBody(
    ComponentPose=@n6
    Alpha=@n5
    OverrideWorldGravity="0,0,-980"
    SimSpaceSettings=(WorldAlpha=1,VelocityScaleZ=0,DampingAlpha=1,MaxLinearVelocity=500,MaxAngularVelocity=3,MaxLinearAcceleration=700,MaxAngularAcceleration=1000,ExternalLinearDragV=(X=0.1,Y=0.1,Z=0),ExternalLinearVelocity=(X=0,Y=0,Z=0),ExternalAngularVelocity=(X=0,Y=0,Z=0))
   ) {
    bClampLinearTranslationLimitToRefPose=true
    bOverrideWorldGravity=true
    overridePhysicsAsset=&/F_MED_BaskIsle_Left_Physics
    simulationSpace=BaseBoneSpace
    alphaBoolBlend {blendOption=Linear}
    baseBoneRef {boneName=thigh_l}
    componentAppliedLinearAccClamp=<0 0 0>
    overrideWorldGravity {z=-980}
    simSpaceSettings {
     maxAngularAcceleration=1000
     maxAngularVelocity=3
     maxLinearAcceleration=700
     maxLinearVelocity=500
     velocityScaleZ=0
     worldAlpha=1
     externalLinearDragV=<0.1 0.1>
    }
   }
   n8 = RigidBody(
    ComponentPose=Local To Component(LocalPose=Use cached pose 'CR_CtrlRig'())
    OverrideWorldGravity="0,0,-980"
    SimSpaceSettings=(WorldAlpha=1,VelocityScaleZ=0,DampingAlpha=1,MaxLinearVelocity=500,MaxAngularVelocity=3,MaxLinearAcceleration=700,MaxAngularAcceleration=1000,ExternalLinearDragV=(X=0.1,Y=0.1,Z=0),ExternalLinearVelocity=(X=0,Y=0,Z=0),ExternalAngularVelocity=(X=0,Y=0,Z=0))
   ) {
    bClampLinearTranslationLimitToRefPose=true
    bOverrideWorldGravity=true
    overridePhysicsAsset=&F_MED_BaskIsle_Right_Physics
    simulationSpace=BaseBoneSpace
    alphaBoolBlend {blendOption=Linear}
    baseBoneRef {boneName=thigh_r}
    componentAppliedLinearAccClamp=<0 0 0>
    overrideWorldGravity {z=-980}
    simSpaceSettings {
     maxAngularAcceleration=1000
     maxAngularVelocity=3
     maxLinearAcceleration=700
     maxLinearVelocity=500
     velocityScaleZ=0
     worldAlpha=1
     externalLinearDragV=<0.1 0.1>
    }
   }
   n9 = RigidBody(
    ComponentPose=@n8
    OverrideWorldGravity="0,0,-980"
    SimSpaceSettings=(WorldAlpha=1,VelocityScaleZ=0,DampingAlpha=1,MaxLinearVelocity=500,MaxAngularVelocity=3,MaxLinearAcceleration=700,MaxAngularAcceleration=1000,ExternalLinearDragV=(X=0.1,Y=0.1,Z=0),ExternalLinearVelocity=(X=0,Y=0,Z=0),ExternalAngularVelocity=(X=0,Y=0,Z=0))
   ) {
    bClampLinearTranslationLimitToRefPose=true
    bOverrideWorldGravity=true
    overridePhysicsAsset=&/F_MED_BaskIsle_Left_Rail_Physics
    simulationSpace=BaseBoneSpace
    alphaBoolBlend {blendOption=Linear}
    baseBoneRef {boneName=thigh_l}
    componentAppliedLinearAccClamp=<0 0 0>
    overrideWorldGravity {z=-980}
    simSpaceSettings {
     maxAngularAcceleration=1000
     maxAngularVelocity=3
     maxLinearAcceleration=700
     maxLinearVelocity=500
     velocityScaleZ=0
     worldAlpha=1
     externalLinearDragV=<0.1 0.1>
    }
   }
   Save cached pose 'PostDyn Cache'(Pose=Blend Poses by bool(BlendPose_0=Component To Local(ComponentPose=@n9), BlendPose_1=Component To Local(ComponentPose=@n7)) {blendType=Linear}) {cacheName=PostDyn Cache}
  ]}
 EventGraph {nodes [
   Event BlueprintUpdateAnimation()
   > n1 = Sequence!(execute) -> (then_0, then_1, then_2:Exec, then_3:Exec, then_4:Exec, then_5:Exec, then_6:Exec)
   > n2 = "Gravity Override State Function (Parameter Asset) (Nativized)"(
    CharacterPartAnimInstance=Self-Reference()
    GameTimeInSeconds=GetGameTimeInSeconds()
    DebugJointName=spine_05
    Parameters=&/F_MED_BaskIsle_Parameters
   )
   > Set Gravity Override(Gravity Override=@n2)
   > Set RBAN Settings Out(RBAN Settings Out=@n2.OutRigidBodyAnimNodeParameters)
   n3 = Branch(execute=@n1.then_1, Condition=OR Boolean(A=Get bIsSprinting(), B=Get bIsTacticalSprinting()))
   > Set isSprinting(isSprinting:double=1) -> (Output_Get:double)
   Set isSprinting(execute=@n3.else, isSprinting=0)
   n4 = Branch(execute=@n1.then_2, Condition=Get bIsOnGrindRail())
   > Set isRailGrinding(isRailGrinding:double=1) -> (Output_Get:double)
   Set isRailGrinding(execute=@n4.else, isRailGrinding=0)
   Set GrindOrBseballSliding(execute=@n1.then_3, GrindOrBseballSliding:bool=OR Boolean(A=Get bIsOnGrindRail(), B=Get bIsBaseballSliding())) -> (Output_Get:bool)
   n5 = Branch(execute=@n1.then_4, Condition=Get bIsMoving2D())
   > Set isMoving2D(isMoving2D:double=1) -> (Output_Get:double)
   Set isMoving2D(execute=@n5.else, isMoving2D=0)
   Branch(execute=@n1.then_5, Condition=Get bIsInFrontEnd())
   > Set Custom Idle is Playing Float(Custom Idle is Playing Float:double=GetCurveValue(self=GetOwnerFrontEndAnimInstance(), CurveName=Frontend_Idle_Bypass)) -> (Output_Get:double)
   n6 = CheckForBackbling()
   Set isWearingBackbling(
    execute=@n1.then_6
    isWearingBackbling=AND Boolean!(A=@n6, B=NOT Boolean(A=@n6.bIsBackblingACape), C:bool=NOT Boolean(A=@n6.bIsBackblingLong)) -> (ReturnValue)
   )
   > n7 = Set Is Wearing Cape(Is Wearing Cape=OR Boolean(A=@n6.bIsBackblingACape, B=@n6.bIsBackblingLong))
   > n8 = Branch(Condition=@n7)
   > Set RB_Value(RB_Value:double=0.9) -> (Output_Get:double)
   Set RB_Value(execute=@n8.else, RB_Value=1)
   Event OnBeginCrouching()
   > Set isCrouching(isCrouching:double=1) -> (Output_Get:double)
   Event OnEndSkydiving()
   > Set isSkyDiving(isSkyDiving:double=0) -> (Output_Get:double)
   Event OnEndCrouching()
   > Set isCrouching(isCrouching=0)
   Event OnBeginSliding() {is=title K2Node_Event}
   > Set isSliding(isSliding:double=1) -> (Output_Get:double)
   Event OnEndSliding() {is=title K2Node_Event}
   > Set isSliding(isSliding=0)
   Event OnBeginBaseballSliding() {is=title K2Node_Event}
   > Set isBaseBallSliding(isBaseBallSliding:double=1) -> (Output_Get:double)
   Event OnEndBaseballSliding() {is=title K2Node_Event}
   > Set isBaseBallSliding(isBaseBallSliding=0)
   Event OnBeginSkydiving()
   > Set isSkyDiving(isSkyDiving=1)
  ]}
]
variables [
 Custom Idle is Playing Float {type=number}
 Gravity Override {type=Vector}
 GrindOrBseballSliding {type=boolean}
 Is Wearing Cape {type=boolean}
 isBaseBallSliding {type=number}
 isCrouching {type=number}
 isMoving2D {type=number}
 isRailGrinding {type=number}
 isSkyDiving {type=number}
 isSliding {type=number}
 isSprinting {type=number}
 isWearingBackbling {type=boolean}
 RB_Value {type=number}
 RBAN Settings Out {type=FortRigidBodyAnimNodeParameters}
]
