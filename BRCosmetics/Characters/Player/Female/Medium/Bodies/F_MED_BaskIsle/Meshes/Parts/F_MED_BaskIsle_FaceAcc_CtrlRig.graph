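# Control Rig for the BaskIsle female medium face accessory, applying pose-driven corrective offsets to hair base bones as the head and clavicles move.
# A five-way sequence feeds ModifyTransforms weights from remapped head rotation, Euler-distance pose matches and spherical pose readers.
# One head-to-right-clavicle distance weight is smoothed with a Kalman filter.
# Both clavicle-driven reader corrections are multiplied by zero, and parachute variable and FK control nodes are unconnected.
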ControlRigBlueprint /BRCosmetics/Characters/Player/Female/Medium/Bodies/F_MED_BaskIsle/Meshes/Parts/F_MED_BaskIsle_FaceAcc_CtrlRig
graphs [
 Multiply_2_1_ContainedGraph {nodes [
   n1 = RigVMFunctionEntryNode()
   RigVMFunctionReturnNode(Result=&RigVMFunction_MathDoubleMul(B=@n1.C, A=&RigVMFunction_MathDoubleMul(A=@n1.A, B=@n1.B).Result).Result)
  ]}
 New Function_ContainedGraph {nodes [
   RigVMFunctionEntryNode()
   RigUnit_GetRelativeTransformForItem = &RigUnit_GetRelativeTransformForItem(Child=(Type=Bone,Name="head"), Parent=(Type=Bone,Name="spine_05"), bChildInitial=False, bParentInitial=False)
   RigVMFunction_MathQuaternionToEuler = &RigVMFunction_MathQuaternionToEuler(RotationOrder=ZYX, Value=@RigUnit_GetRelativeTransformForItem.RelativeTransform.Rotation)
   > RigVMFunctionReturnNode(Result=@RigVMFunction_MathQuaternionToEuler.Result)
  ]}
 RigVMFunctionLibrary {nodes [F_Head Rotation = RigVMCollapseNode()]}
 RigVMFunction_Sequence_1_ContainedGraph {nodes [
   n1 = RigVMFunctionEntryNode()
   RigVMFunction_Sequence = &RigVMFunction_Sequence(ExecuteContext=@n1.ExecuteContext)
   n2 = &RigVMFunction_Sequence(ExecuteContext=@RigVMFunction_Sequence.B)
   n3 = &RigVMFunction_Sequence(ExecuteContext=@n2.B)
   n4 = &RigVMFunction_Sequence(ExecuteContext=@n3.B)
   RigVMFunctionReturnNode(A=@RigVMFunction_Sequence.A, B=@n2.A, C=@n3.A, D=@n4.A, E=@n4.B)
  ]}
 RigVMModel {nodes [
   RigUnit_BeginExecution = &RigUnit_BeginExecution()
   RigVMFunction_Sequence_1 = RigVMAggregateNode(ExecuteContext=@RigUnit_BeginExecution.ExecutePin)
   n1 = &RigUnit_SphericalPoseReader(
    ActiveRegionScaleFactors=(PositiveWidth=1,NegativeWidth=1,PositiveHeight=1,NegativeHeight=1)
    ActiveRegionSize=0.01
    Debug=(bDrawDebug=false,bDraw2D=False,bDrawLocalAxes=False,DebugScale=25,DebugSegments=20,DebugThickness=0.25)
    DriverAxis=(X=0,Y=1,Z=0)
    DriverItem=(Type=Bone,Name="head")
    FalloffRegionScaleFactors=(PositiveWidth=1,NegativeWidth=1,PositiveHeight=1,NegativeHeight=1)
    FalloffSize=0.1
    FlipHeightScaling=False
    FlipWidthScaling=False
    OptionalParentItem=(Type=Bone,Name="clavicle_r")
    RotationOffset=(X=0,Y=-87,Z=-5)
    ExecutePin=@RigVMFunction_Sequence_1.D
   )
   > &RigUnit_ModifyTransforms(
    ItemToModify="((Item=(Type=Bone,Name=\"base_hair_side_rt\"),Transform=(Rotation=(X=0,Y=0.2588,Z=0,W=-0.9659),Translation=(X=-2,Y=0,Z=0),Scale3D=(X=1,Y=1,Z=1))),(Item=(Type=Bone,Name=\"base_hair_rt\"),Transform=(Rotation=(X=0,Y=0,Z=-0.0872,W=0.9962),Translation=(X=-5,Y=0,Z=0 ...(truncated)"
    WeightMinimum=0
    Weight=@n1.OutputParam
   )
   n2 = &RigVMFunction_MathVectorDistance(
    A=&RigVMFunction_MathQuaternionToEuler(
     RotationOrder=YZX
     Value=&RigUnit_GetRelativeTransformForItem(Child=(Type=Bone,Name="clavicle_r"), Parent=(Type=Bone,Name="head"), bChildInitial=False, bParentInitial=False).RelativeTransform.Rotation
    ).Result
    B=/Script/RigVM.RigVMFunction_MathVectorMake(X=-127.1, Y=-156, Z=-61.94) {is=struct RigVMUnitNode}.Result
   )
   > &RigUnit_ModifyTransforms(
    ItemToModify=((Item=(Type=Bone,Name="base_hair_lf"),Transform=(Rotation=(X=0,Y=0,Z=0,W=1),Translation=(X=0,Y=0,Z=0),Scale3D=(X=1,Y=1,Z=1))),(Item=(Type=Bone,Name="base_hair_cr"),Transform=(Rotation=(X=0,Y=0,Z=0,W=1),Translation=(X=0,Y=0,Z=0),Scale3D=(X=1,Y=1,Z=1))))
    WeightMinimum=0
    Weight=&RigVMFunction_KalmanFloat(
     BufferSize=2
     Value=&RigVMFunction_MathDoubleRemap(SourceMaximum=3, SourceMinimum=20, TargetMaximum=1, TargetMinimum=0, Value=@n2.Result).Result
    ) {template="KalmanFilter::Execute(in Value,in BufferSize,out Result)"}.Result
   )
   n3 = &RigUnit_SphericalPoseReader(
    ActiveRegionScaleFactors=(PositiveWidth=1,NegativeWidth=1,PositiveHeight=1,NegativeHeight=1)
    ActiveRegionSize=0.1
    Debug=(bDrawDebug=false,bDraw2D=false,bDrawLocalAxes=False,DebugScale=25,DebugSegments=20,DebugThickness=0.25)
    DriverAxis=(X=1,Y=0,Z=0)
    DriverItem=(Type=Bone,Name="clavicle_l")
    FalloffRegionScaleFactors=(PositiveWidth=1,NegativeWidth=1,PositiveHeight=1,NegativeHeight=1)
    FalloffSize=0.2
    FlipHeightScaling=False
    FlipWidthScaling=False
    OptionalParentItem=(Type=Bone,Name="None")
    RotationOffset=(X=0,Y=-30,Z=0)
    ExecutePin=@RigVMFunction_Sequence_1.E
   )
   Multiply = &RigVMFunction_MathDoubleMul(B=0, A=@n3.OutputParam)
   > &RigUnit_ModifyTransforms(
    ItemToModify="((Item=(Type=Bone,Name=\"base_hair_side_lf\"),Transform=(Rotation=(X=0,Y=0,Z=-0.1305,W=0.9914),Translation=(X=-2.5,Y=-2.5,Z=0),Scale3D=(X=1,Y=1,Z=1))),(Item=(Type=Bone,Name=\"base_hair_lf\"),Transform=(Rotation=(X=0,Y=0,Z=0.1305,W=0.9914),Translation=(X=1,Y=0, ...(truncated)"
    WeightMinimum=0
    Weight=@Multiply.Result
   )
   > n4 = &RigUnit_SphericalPoseReader(
    ActiveRegionScaleFactors=(PositiveWidth=1,NegativeWidth=1,PositiveHeight=1,NegativeHeight=1)
    ActiveRegionSize=0.1
    Debug=(bDrawDebug=false,bDraw2D=False,bDrawLocalAxes=False,DebugScale=25,DebugSegments=20,DebugThickness=0.25)
    DriverAxis=(X=-1,Y=0,Z=0)
    DriverItem=(Type=Bone,Name="clavicle_r")
    FalloffRegionScaleFactors=(PositiveWidth=1,NegativeWidth=1,PositiveHeight=1,NegativeHeight=1)
    FalloffSize=0.2
    FlipHeightScaling=False
    FlipWidthScaling=False
    OptionalParentItem=(Type=Bone,Name="None")
    RotationOffset=(X=0,Y=150,Z=0)
   )
   > &RigUnit_ModifyTransforms(
    ItemToModify="((Item=(Type=Bone,Name=\"base_hair_side_rt\"),Transform=(Rotation=(X=0,Y=0,Z=0.1305,W=0.9914),Translation=(X=-2.5,Y=2.5,Z=0),Scale3D=(X=1,Y=1,Z=1))),(Item=(Type=Bone,Name=\"base_hair_rt\"),Transform=(Rotation=(X=0,Y=0,Z=0.1132,W=0.9936),Translation=(X=-1,Y=0,Z ...(truncated)"
    WeightMinimum=0
    Weight=&RigVMFunction_MathDoubleMul(B=0, A=@n4.OutputParam).Result
   )
   n5 = &RigVMFunction_MathQuaternionToEuler(
    RotationOrder=XZY
    Value=&RigUnit_GetRelativeTransformForItem(Child=(Type=Bone,Name="head"), Parent=(Type=Bone,Name="root"), bChildInitial=False, bParentInitial=False).RelativeTransform.Rotation
   )
   n6 = &RigVMFunction_MathVectorDistance(
    A=&RigVMFunction_MathQuaternionToEuler(
     RotationOrder=XZY
     Value=&RigUnit_GetRelativeTransformForItem(Child=(Type=Bone,Name="head"), Parent=(Type=Bone,Name="clavicle_r"), bChildInitial=False, bParentInitial=False).RelativeTransform.Rotation
    ).Result
    B=&RigVMFunction_MathVectorMake(X=130.54, Y=161.807, Z=18.695).Result
   )
   &RigUnit_ModifyTransforms(
    ItemToModify="((Item=(Type=Bone,Name=\"base_hair_cr\"),Transform=(Rotation=(X=0,Y=0.0436,Z=0,W=0.999),Translation=(X=-3,Y=0,Z=0),Scale3D=(X=1,Y=1,Z=1))),(Item=(Type=Bone,Name=\"base_hair_rt\"),Transform=(Rotation=(X=0,Y=0,Z=0,W=1),Translation=(X=-3,Y=0,Z=0),Scale3D=(X=1,Y=1 ...(truncated)"
    WeightMinimum=0
    Weight=&RigVMFunction_MathDoubleMul(
     B=&RigVMFunction_MathDoubleRemap(SourceMaximum=-55, SourceMinimum=-10, TargetMaximum=1, TargetMinimum=0, Value=@n5.Result.Z).Result
     A=&RigVMFunction_MathDoubleRemap(SourceMaximum=5, SourceMinimum=40, TargetMaximum=1, TargetMinimum=0, Value=@n6.Result).Result
    ).Result
    ExecutePin=@RigVMFunction_Sequence_1.C
   )
   F_Head Rotation = RigVMFunctionReferenceNode()
   Remap = &RigVMFunction_MathDoubleRemap(SourceMaximum=-80, SourceMinimum=0, TargetMaximum=1, TargetMinimum=0, Value=@"F_Head Rotation".Result.Z)
   &RigUnit_ModifyTransforms(
    ItemToModify="((Item=(Type=Bone,Name=\"base_hair_lf\"),Transform=(Rotation=(X=0,Y=-0.2588,Z=0,W=0.9659),Translation=(X=-4,Y=0,Z=0),Scale3D=(X=1,Y=1,Z=1))),(Item=(Type=Bone,Name=\"base_hair_cr\"),Transform=(Rotation=(X=0,Y=-0.2588,Z=0,W=0.9659),Translation=(X=-4,Y=0,Z=0),Sca ...(truncated)"
    WeightMinimum=0
    Weight=@Remap.Result
    ExecutePin=@RigVMFunction_Sequence_1.A
   )
   F_Head Rotation_4 = RigVMFunctionReferenceNode()
   Power = /Script/RigVM.RigVMFunction_MathDoublePow(
    B=0.7
    A=&RigVMFunction_MathDoubleRemap(SourceMaximum=-55, SourceMinimum=0, TargetMaximum=1, TargetMinimum=0, Value=@"F_Head Rotation_4".Result.Z).Result
   ) {is=struct RigVMUnitNode, template="Power::Execute(in A,in B,out Result)"}
   &RigUnit_ModifyTransforms(
    ItemToModify="((Item=(Type=Bone,Name=\"base_hair_lf\"),Transform=(Rotation=(X=0,Y=0,Z=0,W=1),Translation=(X=0,Y=0,Z=0),Scale3D=(X=1,Y=1,Z=1))),(Item=(Type=Bone,Name=\"base_hair_cr\"),Transform=(Rotation=(X=0,Y=0,Z=0,W=1),Translation=(X=0,Y=0,Z=0),Scale3D=(X=1,Y=1,Z=1))),(It ...(truncated)"
    WeightMinimum=0
    Weight=@Power.Result
    ExecutePin=@RigVMFunction_Sequence_1.B
   )
   > &RigUnit_ModifyTransforms(
    ItemToModify="((Item=(Type=Bone,Name=\"base_hair_side_rt\"),Transform=(Rotation=(X=0,Y=-0.0872,Z=0,W=0.9962),Translation=(X=-2,Y=0,Z=-2),Scale3D=(X=1,Y=1,Z=1))),(Item=(Type=Bone,Name=\"base_hair_rt\"),Transform=(Rotation=(X=0,Y=0,Z=0,W=1),Translation=(X=-3.5,Y=0,Z=0),Scale3 ...(truncated)"
    Weight=0
    WeightMinimum=0
   )
   n7 = &RigVMFunction_MathQuaternionToEuler(
    RotationOrder=ZYX
    Value=&RigUnit_GetRelativeTransformForItem(Child=(Type=Bone,Name="head"), Parent=(Type=Null,Name="head_null"), bChildInitial=False, bParentInitial=False).RelativeTransform.Rotation
   )
   n8 = &RigUnit_GetRelativeTransformForItem(Child=(Type=Bone,Name="clavicle_r"), Parent=(Type=Null,Name="clavicle_r_null"), bChildInitial=False, bParentInitial=False)
   Multiply_2 = RigVMAggregateNode(
    A=&RigVMFunction_MathDoubleRemap(SourceMaximum=-29, SourceMinimum=-5, TargetMaximum=1, TargetMinimum=0, Value=@n7.Result.Y).Result
    B=&RigVMFunction_MathDoubleRemap(SourceMaximum=-20, SourceMinimum=5, TargetMaximum=1, TargetMinimum=0, Value=@n7.Result.Z).Result
    C=&RigVMFunction_MathDoubleRemap(
     SourceMaximum=-14
     SourceMinimum=0
     TargetMaximum=1
     TargetMinimum=0
     bClamp=false
     Value=&RigVMFunction_MathQuaternionToEuler(RotationOrder=ZYX, Value=@n8.RelativeTransform.Rotation).Result.Y
    ).Result
   )
   &RigUnit_PrepareForExecution()
   RigVMFunction_MathVectorMake = &RigVMFunction_MathVectorMake(X=-141.5, Y=-138.5, Z=-9.3)
   RigVMFunction_MathVectorDistance = &RigVMFunction_MathVectorDistance(
    A=&RigVMFunction_MathQuaternionToEuler(
     RotationOrder=YZX
     Value=&RigUnit_GetRelativeTransformForItem(Child=(Type=Bone,Name="clavicle_r"), Parent=(Type=Bone,Name="head"), bChildInitial=False, bParentInitial=False).RelativeTransform.Rotation
    ).Result
    B=@RigVMFunction_MathVectorMake.Result
   )
   &RigVMFunction_MathDoubleRemap(SourceMaximum=3, SourceMinimum=20, TargetMaximum=1, TargetMinimum=0, Value=@RigVMFunction_MathVectorDistance.Result)
   RigVMVariableNode() {variable=isParachuteOpened}
   RigVMVariableNode() {variable=isParachuteOpened}
   addFKCtrl = RigVMFunctionReferenceNode(
    Custom_FKCtrl=()
    ExecuteContext=()
    Options_FKCtrl="(pelvis_1_#1=false,spine_3_#2=false,clavicle_5_#3=true,arm_7_#4=false,leg_9_#5=false,head_11_46DC15EF465CD ...(truncated)"
    Shape_Settings="(pelvis_2_#6=(bVisible=True,Name=\"Hexagon_Thin\",Color=(R=0.0117,G=0.8281,B=1,A=1),Transform=(Rotation=(X=0,Y=-0.7071,Z=0,W=0.7071),Translation=(X=2,Y=0,Z=0),Scale3D=(X=5,Y=5,Z=7))),spine_4_#7=(bVi ...(truncated)"
   )
   setFKCtrl_1 = RigVMFunctionReferenceNode(ExecuteContext=(), setFK_Ctrl=false)
  ]}
]
hierarchy [
 root {loc=<0 0 0>, rot=<0 0 0 1>}
 pelvis {parent=root, loc=<0 -0.5301 98.0571>, rot=<0.0127 0.707 -0.0127 -0.707>}
 spine_01 {parent=pelvis, loc=<6.1208 -0.9328 0>, rot=<0 0 0.034 -0.9994>}
 spine_02 {parent=spine_01, loc=<6.5997 -0.2095 0>, rot=<0 0 0.0238 -0.9997>}
 spine_03 {parent=spine_02, loc=<6.4914 -0.1237 0>, rot=<0 0 0.0149 -0.9999>}
 spine_04 {parent=spine_03, loc=<6.9598 0.0794 0>, rot=<0 0 -0.0132 -0.9999>}
 spine_05 {parent=spine_04, loc=<6.9987 0 0>, rot=<0 0 -0.0015 -1>}
 neck_01 {parent=spine_05, loc=<11.3138 0.1529 0>, rot=<0 0 -0.2003 -0.9797>}
 neck_02 {parent=neck_01, loc=<4.5627 0 0>, rot=<0 0 0.0504 -0.9987>}
 head {parent=neck_02, loc=<5.3914 0 0>, rot=<0 0 0.115 -0.9934>}
 base_hair_Front_A_01 {parent=head, loc=<13.8884 9.8241 2.8568>, rot=<0.3091 0.4252 0.8483 0.0637>}
 dyn_hair_Front_A_01 {parent=base_hair_Front_A_01, loc=<0 0 0>, rot=<0 0 0 1>}
 dyn_hair_Front_A_02 {parent=dyn_hair_Front_A_01, loc=<3.7518 0 0>, rot=<-0.0183 -0.3263 -0.1252 -0.9368>}
 base_hair_Front_B_01 {parent=head, loc=<13.7496 10.711 -1.7865>, rot=<-0.1883 -0.7549 -0.6204 -0.099>}
 dyn_hair_Front_B_01 {parent=base_hair_Front_B_01, loc=<0 0 0>, rot=<0 0 0 1>}
 dyn_hair_Front_B_02 {parent=dyn_hair_Front_B_01, loc=<4.7796 0 0>, rot=<-0.0154 -0.3658 -0.1378 -0.9203>}
 base_hair_Front_C_01 {parent=head, loc=<12.9614 9.3684 -5.0789>, rot=<-0.1146 -0.9217 -0.2987 -0.2192>}
 dyn_hair_Front_C_01 {parent=base_hair_Front_C_01, loc=<0 0 0>, rot=<0 0 0 1>}
 dyn_hair_Front_C_02 {parent=dyn_hair_Front_C_01, loc=<3.1736 0 0>, rot=<-0.0248 -0.398 0.003 -0.917>}
 base_hair_side_lf {parent=head, loc=<8.1067 -1.0341 -7.4052>, rot=<-0.0853 -0.6939 0.7078 -0.1011>}
 dyn_hair_side_01_lf {parent=base_hair_side_lf, loc=<0 0 0>, rot=<0 0 0 1>}
 dyn_hair_side_02_lf {parent=dyn_hair_side_01_lf, loc=<8.2561 0 0>, rot=<0 0 0 1>}
 base_hair_side_rt {parent=head, loc=<8.1066 -1.0337 7.405>, rot=<0.0969 -0.6939 0.7079 0.0893>}
 dyn_hair_side_01_rt {parent=base_hair_side_rt, loc=<0 0 0>, rot=<0 0 0 1>}
 dyn_hair_side_02_rt {parent=dyn_hair_side_01_rt, loc=<8.2561 0 0>, rot=<0 0 0 1>}
 base_hair_lf {parent=head, loc=<8.1532 -6.664 -5.2664>, rot=<0.0761 -0.7092 0.7002 -0.0319>}
 dyn_hair_01_lf {parent=base_hair_lf, loc=<0 0 0>, rot=<0 0 0 1>}
 dyn_hair_02_lf {parent=dyn_hair_01_lf, loc=<8.2561 0 0>, rot=<0 0 0 1>}
 base_hair_cr {parent=head, loc=<8.1562 -7.025 0>, rot=<0.0761 -0.7092 0.7002 -0.0319>}
 dyn_hair_01_cr {parent=base_hair_cr, loc=<0 0 0>, rot=<0 0 0 1>}
 dyn_hair_02_cr {parent=dyn_hair_01_cr, loc=<8.2561 0 0>, rot=<0 0 0 1>}
 base_hair_rt {parent=head, loc=<8.1532 -6.664 5.266>, rot=<0.0761 -0.7092 0.7002 -0.0319>}
 dyn_hair_01_rt {parent=base_hair_rt, loc=<0 0 0>, rot=<0 0 0 1>}
 dyn_hair_02_rt {parent=dyn_hair_01_rt, loc=<8.2561 0 0>, rot=<0 0 0 1>}
 base_hair_l_A_01 {parent=head, loc=<1.0968 1.7393 -9.7393>, rot=<-0.1828 -0.9518 -0.2339 0.0764>}
 dyn_hair_l_A_01 {parent=base_hair_l_A_01, loc=<0 0 0>, rot=<0 0 0 1>}
 dyn_hair_l_A_02 {parent=dyn_hair_l_A_01, loc=<4.3175 0 0>, rot=<-0.0913 -0.0772 -0.3264 -0.9376>}
 base_hair_l_B_01 {parent=head, loc=<1.1534 -5.1057 -9.7393>, rot=<-0.1212 -0.9571 0.2587 0.0485>}
 dyn_hair_l_B_01 {parent=base_hair_l_B_01, loc=<0 0 0>, rot=<0 0 0 1>}
 dyn_hair_l_B_02 {parent=dyn_hair_l_B_01, loc=<5.2105 0 0>, rot=<-0.0284 -0.3267 -0.0963 -0.9398>}
 base_hair_r_A_01 {parent=head, loc=<1.0964 1.7393 9.7394>, rot=<0.9518 -0.1828 0.0764 0.2339>}
 dyn_hair_r_A_01 {parent=base_hair_r_A_01, loc=<0 0 0>, rot=<0 0 0 1>}
 dyn_hair_r_A_02 {parent=dyn_hair_r_A_01, loc=<-4.3174 -0.0001 0>, rot=<-0.0913 -0.0772 -0.3264 -0.9376>}
 base_hair_r_B_01 {parent=head, loc=<1.153 -5.1057 9.7394>, rot=<0.9571 -0.1212 0.0485 -0.2587>}
 dyn_hair_r_B_01 {parent=base_hair_r_B_01, loc=<0 0 0>, rot=<0 0 0 1>}
 dyn_hair_r_B_02 {parent=dyn_hair_r_B_01, loc=<-5.2099 -0.0001 0.0001>, rot=<-0.0284 -0.3267 -0.0963 -0.9398>}
 base_hair_l_C_01 {parent=head, loc=<0.8656 -9.7599 -5.9121>, rot=<-0.0436 -0.8597 0.5017 0.0857>}
 dyn_hair_l_C_01 {parent=base_hair_l_C_01, loc=<0 0 0>, rot=<0 0 0 1>}
 dyn_hair_l_C_02 {parent=dyn_hair_l_C_01, loc=<5.3577 0 0>, rot=<0 0 0 1>}
 base_hair_r_C_01 {parent=head, loc=<0.8655 -9.7599 5.9121>, rot=<0.8597 -0.0436 0.0857 -0.5017>}
 dyn_hair_r_C_01 {parent=base_hair_r_C_01, loc=<0 0 0>, rot=<0 0 0 1>}
 dyn_hair_r_C_02 {parent=dyn_hair_r_C_01, loc=<-5.3574 0 0>, rot=<0 0 0 1>}
 base_hair_C_01 {parent=head, loc=<0.7628 -11.3277 0>, rot=<-0.0844 -0.702 0.702 0.0844>}
 dyn_hair_C_01 {parent=base_hair_C_01, loc=<0 0 0>, rot=<0 0 0 1>}
 dyn_hair_C_02 {parent=dyn_hair_C_01, loc=<5.0546 0 0>, rot=<0 0 0 1>}
 clavicle_l {parent=spine_05, loc=<5.4831 -0.1053 -3.2946>, rot=<0.0465 -0.7289 -0.009 -0.6829>}
 upperarm_l {parent=clavicle_l, loc=<9.8658 -1.2136 -0.4105>, rot=<0.0164 -0.3245 -0.0269 -0.9454>}
 upperarm_twist_01_l {parent=upperarm_l, loc=<7.2135 -0.1134 0>, rot=<0 0 0 1>}
 deform_upperarm_bk_base_l {parent=upperarm_twist_01_l, loc=<-6.4053 -4.3746 0.314>, rot=<0 0 0 -1>}
 deform_upperarm_bk_l {parent=deform_upperarm_bk_base_l, loc=<0 0 0>, rot=<0 0 0 1>}
 deform_upperarm_fr_base_l {parent=upperarm_twist_01_l, loc=<-5.6312 5.092 0.3354>, rot=<0 0 0 -1>}
 deform_upperarm_fr_l {parent=deform_upperarm_fr_base_l, loc=<0 0 0>, rot=<0 0 0 1>}
 deform_upperarm_up_base_l {parent=upperarm_twist_01_l, loc=<-8.891 0.4825 3.9517>, rot=<0 0 0 -1>}
 deform_upperarm_up_l {parent=deform_upperarm_up_base_l, loc=<0 0 0>, rot=<0 0 0 1>}
 clavicle_r {parent=spine_05, loc=<5.4831 -0.1053 3.2946>, rot=<0.7289 0.0465 -0.6829 0.009>}
 upperarm_r {parent=clavicle_r, loc=<-9.8658 1.2136 0.4105>, rot=<0.0164 -0.3245 -0.0269 -0.9454>}
 upperarm_twist_01_r {parent=upperarm_r, loc=<-7.2135 0.1134 0>, rot=<0 0 0 -1>}
 deform_upperarm_bk_base_r {parent=upperarm_twist_01_r, loc=<6.4053 4.3746 -0.314>, rot=<0 0 0 -1>}
 deform_upperarm_bk_r {parent=deform_upperarm_bk_base_r, loc=<0 0 0>, rot=<0 0 0 1>}
 deform_upperarm_fr_base_r {parent=upperarm_twist_01_r, loc=<5.6312 -5.092 -0.3354>, rot=<0 0 0 -1>}
 deform_upperarm_fr_r {parent=deform_upperarm_fr_base_r, loc=<0 0 0>, rot=<0 0 0 1>}
 deform_upperarm_up_base_r {parent=upperarm_twist_01_r, loc=<8.891 -0.4825 -3.9517>, rot=<0 0 0 -1>}
 deform_upperarm_up_r {parent=deform_upperarm_up_base_r, loc=<0 0 0>, rot=<0 0 0 1>}
 head_null {parent=spine_05, type=Null, loc=<20.6573 3.5489 0>, rot=<0 0 -0.036 -0.9994>}
 clavicle_r_null {parent=spine_05, type=Null, loc=<5.483 -0.1053 3.2946>, rot=<0.7289 0.0465 -0.6829 0.009>}
 deform_pec_aim_base_r {parent=spine_04, loc=<3.2452 9.3309 1.2237>, rot=<-0.4033 -0.0603 0.8507 0.3318>}
 deform_pec_aim_r {parent=deform_pec_aim_base_r, loc=<0 0 0>, rot=<0 0 0 1>}
 deform_pec_base_r {parent=deform_pec_aim_r, loc=<-9.3667 -0.0003 -0.0004>, rot=<0 0 0 1>}
 deform_pec_r {parent=deform_pec_base_r, loc=<0 0 0>, rot=<0 0 0 1>}
]
variables [
 SphericalPoseReader_2_OutputParam {type=float, default=0}
 SphericalPoseReader_3_OutputParam {type=float, default=0}
 isParachuteOpened {type=double, default=0}
]
