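# Anim Blueprint for the F_MED_BaskIsle face accessory part, driving secondary hair motion and a glasses animation layer.
# The parent mesh pose runs through a Control Rig with a parachute input (0 when open) and additive hair-bone offsets.
# One of two RigidBody sims, chosen by non-cape, non-long backbling, feeds front-hair AnimDynamics.
# DefaultSlot animation layers onto the Glasses bone, weighted by a front-end idle curve; computed gravity overrides appear unwired.
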AnimBlueprint /BRCosmetics/Characters/Player/Female/Medium/Bodies/F_MED_BaskIsle/Meshes/Parts/F_MED_BaskIsle_FaceAcc_AnimBP
target_skeleton=&/F_MED_BaskIsle_FaceAcc_Skeleton
graphs [
 AnimGraph {nodes [
   n1 = RigidBody() {
    overridePhysicsAsset=&/F_MED_BaskIsle_FaceAcc_Hair_Physics
    alphaBoolBlend {blendOption=Linear}
   }
   Component To Local(ComponentPose=@n1)
   RigidBody!(
    ComponentPose
    Alpha
    OverrideWorldGravity:Vector="0,0,-980"
    ExternalForce
    SimSpaceSettings:Sim Space Settings Structure=(WorldAlpha=1,VelocityScaleZ=0,DampingAlpha=1,MaxLinearVelocity=500,MaxAngularVelocity=3,MaxLinearAcceleration=700,MaxAngularAcceleration=1000,ExternalLinearDragV=(X=0.1,Y=0.1,Z=0),ExternalLinearVelocity=(X=0,Y=0,Z=0),ExternalAngularVelocity=(X=0,Y=0,Z=0))
   ) -> (Pose) {
    bClampLinearTranslationLimitToRefPose=true
    bOverrideWorldGravity=true
    overridePhysicsAsset=&F_MED_BaskIsle_FaceAcc_Hair_Physics
    simulationSpace=BaseBoneSpace
    alphaBoolBlend {blendOption=Linear}
    baseBoneRef {boneName=spine_05}
    componentAppliedLinearAccClamp=<0 0 0>
    overrideWorldGravity {z=-980}
    simSpaceSettings {
     maxAngularAcceleration=1000
     maxAngularVelocity=3
     maxLinearAcceleration=700
     maxLinearVelocity=500
     velocityScaleZ=0
     worldAlpha=1
     externalLinearDragV=<0.1 0.1>
    }
   }
   Output Pose(Result=Layered blend per bone(
     BasePose=Use cached pose 'Cache PostDyn'()
     BlendPoses_0=Slot 'DefaultSlot'()
     BlendWeights_0=Get Custom Idle is Playing Float() -> (Custom Idle is Playing Float:double)
    ) {layerSetup [{branchFilters [{boneName=Glasses}]}]}) {layerGroup=None}
   n2 = Control Rig!(
    Source=Copy Pose From Mesh() {bUseAttachedParent=true}
    Alpha
    isParachuteOpened:double=Get ispParachuteOpened() -> (ispParachuteOpened:double)
   ) -> (Pose) {
    alphaBoolBlend {blendOption=Linear}
    controlRigAssetReference {blueprintRigClass=&/F_MED_BaskIsle_FaceAcc_CtrlRig.F_MED_BaskIsle_FaceAcc_CtrlRig_C}
   }
   n3 = "Transform (Modify) Bone - Bone: base_hair_side_rt"!(ComponentPose=Local To Component(LocalPose=@n2), bAlphaBoolEnabled:bool=True, Translation="0,0,0", Rotation="0,10,0", Scale="1,1,1") -> (Pose) {
    alphaInputType=Bool
    rotationMode=BMM_Additive
    rotationSpace=BCS_BoneSpace
    alphaBoolBlend {blendOption=Linear}
    boneToModify {boneName=base_hair_side_rt}
    rotation {yaw=10}
   }
   n4 = "Transform (Modify) Bone - Bone: base_hair_side_lf"!(ComponentPose=@n3, bAlphaBoolEnabled:bool=True, Translation="0,0,0", Rotation="-20,-30,0", Scale="1,1,1") -> (Pose) {
    alphaInputType=Bool
    rotationMode=BMM_Additive
    rotationSpace=BCS_BoneSpace
    alphaBoolBlend {blendOption=Linear}
    boneToModify {boneName=base_hair_side_lf}
    rotation {pitch=-20, yaw=-30}
   }
   n5 = "Transform (Modify) Bone - Bone: base_hair_lf"!(ComponentPose=@n4, bAlphaBoolEnabled:bool=True, Translation="0,0,0", Rotation="0,-5,0", Scale="1,1,1") -> (Pose) {
    alphaInputType=Bool
    rotationMode=BMM_Additive
    rotationSpace=BCS_BoneSpace
    alphaBoolBlend {blendOption=Linear}
    boneToModify {boneName=base_hair_lf}
    rotation {yaw=-5}
   }
   Save cached pose 'CR_Start'(Pose=Component To Local(ComponentPose=@n5)) {cacheName=CR_Start}
   n6 = RigidBody(
    ComponentPose=Local To Component(LocalPose=Use cached pose 'CR_Start'())
    OverrideWorldGravity="0,0,0"
    SimSpaceSettings=(WorldAlpha=0,VelocityScaleZ=1,DampingAlpha=1,MaxLinearVelocity=10000,MaxAngularVelocity=10000,MaxLinearAcceleration=10000,MaxAngularAcceleration=10000,ExternalLinearDragV=(X=0,Y=0,Z=0),ExternalLinearVelocity=(X=0,Y=0,Z=0),ExternalAngularVelocity=(X=0,Y=0,Z=0))
   ) {
    bClampLinearTranslationLimitToRefPose=true
    bOverrideWorldGravity=true
    overridePhysicsAsset=&/F_MED_BaskIsle_FaceAcc_Hair_WearingBackpack_Physics
    simulationSpace=BaseBoneSpace
    alphaBoolBlend {blendOption=Linear}
    baseBoneRef {boneName=spine_05}
    overrideWorldGravity {z=-980}
   }
   n7 = RigidBody(
    ComponentPose=Local To Component(LocalPose=Use cached pose 'CR_Start'())
    OverrideWorldGravity="0,0,0"
    SimSpaceSettings=(WorldAlpha=0.4,VelocityScaleZ=1,DampingAlpha=1,MaxLinearVelocity=10000,MaxAngularVelocity=10000,MaxLinearAcceleration=10000,MaxAngularAcceleration=10000,ExternalLinearDragV=(X=0,Y=0,Z=0),ExternalLinearVelocity=(X=0,Y=0,Z=0),ExternalAngularVelocity=(X=0,Y=0,Z=0))
   ) {
    bClampLinearTranslationLimitToRefPose=true
    bOverrideWorldGravity=true
    overridePhysicsAsset=&F_MED_BaskIsle_FaceAcc_Hair_Physics
    simulationSpace=BaseBoneSpace
    alphaBoolBlend {blendOption=Linear}
    baseBoneRef {boneName=spine_05}
    overrideWorldGravity {z=-980}
    simSpaceSettings {worldAlpha=0.4}
   }
   Save cached pose 'CR_Backpack_Pose'(Pose=Blend Poses by bool(BlendPose_0=Component To Local(ComponentPose=@n6), BlendPose_1=Component To Local(ComponentPose=@n7), bActiveValue=Get isWearingBackbling()) {blendType=Linear}) {cacheName=CR_Backpack_Pose}
   n8 = Anim Dynamics - Bone: dyn_hair_Front_A_01(ComponentPose=Local To Component(LocalPose=Use cached pose 'CR_Backpack_Pose'()), Alpha=1, ExternalForce="0,0,0") {
    angularDampingOverride=1
    angularSpringConstant=20000
    bAngularSpring=true
    bOverrideAngularDamping=true
    bUseGravityOverride=true
    gravityScale=0.35
    simulationSpace=BoneRelative
    alphaBoolBlend {blendOption=Linear}
    boundBone {boneName=dyn_hair_Front_A_01}
    componentAppliedLinearAccClamp=<3000 3000 3000>
    componentLinearAccScale=<1 1 1>
    componentLinearVelScale=<1 1 1>
    gravityOverride {z=-980}
    physicsBodyDefinitions [{
      boundBone {boneName=dyn_hair_Front_A_01}
      boxExtents=<20 30 30>
      constraintSetup {angularLimitsMax=<5 0 5>, angularLimitsMin=<-5 -5 -5>, angularTarget=<1 0 0>}
      localJointOffset=<4 0 0>
     }]
    relativeSpaceBone {boneName=pelvis}
    simSpaceSettings {maxAngularAcceleration=30, maxAngularVelocity=5, simSpaceAngularAlpha=0.7}
   }
   n9 = Anim Dynamics - Bone: dyn_hair_Front_A_02(ComponentPose=@n8, Alpha=1, ExternalForce="0,0,0") {
    angularDampingOverride=1
    angularSpringConstant=20000
    bAngularSpring=true
    bOverrideAngularDamping=true
    bUseGravityOverride=true
    gravityScale=0.35
    simulationSpace=BoneRelative
    alphaBoolBlend {blendOption=Linear}
    boundBone {boneName=dyn_hair_Front_A_02}
    componentAppliedLinearAccClamp=<3000 3000 3000>
    componentLinearAccScale=<1 1 1>
    componentLinearVelScale=<1 1 1>
    gravityOverride {z=-980}
    physicsBodyDefinitions [{
      boundBone {boneName=dyn_hair_Front_A_02}
      boxExtents=<20 30 30>
      constraintSetup {angularLimitsMax=<5 0 5>, angularLimitsMin=<-5 -5 -5>, angularTarget=<1 0 0>}
      localJointOffset=<4 0 0>
     }]
    relativeSpaceBone {boneName=pelvis}
    simSpaceSettings {maxAngularAcceleration=30, maxAngularVelocity=5, simSpaceAngularAlpha=0.7}
   }
   n10 = Anim Dynamics - Bone: dyn_hair_Front_B_01(ComponentPose=@n9, Alpha=1, ExternalForce="0,0,0") {
    angularDampingOverride=1
    angularSpringConstant=20000
    bAngularSpring=true
    bOverrideAngularDamping=true
    bUseGravityOverride=true
    gravityScale=0.35
    simulationSpace=BoneRelative
    alphaBoolBlend {blendOption=Linear}
    boundBone {boneName=dyn_hair_Front_B_01}
    componentAppliedLinearAccClamp=<3000 3000 3000>
    componentLinearAccScale=<1 1 1>
    componentLinearVelScale=<1 1 1>
    gravityOverride {z=-980}
    physicsBodyDefinitions [{
      boundBone {boneName=dyn_hair_Front_B_01}
      boxExtents=<20 30 30>
      constraintSetup {angularLimitsMax=<5 0 5>, angularLimitsMin=<-5 -5 -5>, angularTarget=<1 0 0>}
      localJointOffset=<4 0 0>
     }]
    relativeSpaceBone {boneName=pelvis}
    simSpaceSettings {maxAngularAcceleration=30, maxAngularVelocity=5, simSpaceAngularAlpha=0.7}
   }
   n11 = Anim Dynamics - Bone: dyn_hair_Front_B_02(ComponentPose=@n10, Alpha=1, ExternalForce="0,0,0") {
    angularDampingOverride=1
    angularSpringConstant=20000
    bAngularSpring=true
    bOverrideAngularDamping=true
    bUseGravityOverride=true
    gravityScale=0.35
    simulationSpace=BoneRelative
    alphaBoolBlend {blendOption=Linear}
    boundBone {boneName=dyn_hair_Front_B_02}
    componentAppliedLinearAccClamp=<3000 3000 3000>
    componentLinearAccScale=<1 1 1>
    componentLinearVelScale=<1 1 1>
    gravityOverride {z=-980}
    physicsBodyDefinitions [{
      boundBone {boneName=dyn_hair_Front_B_02}
      boxExtents=<20 30 30>
      constraintSetup {angularLimitsMax=<5 0 5>, angularLimitsMin=<-5 -5 -5>, angularTarget=<1 0 0>}
      localJointOffset=<4 0 0>
     }]
    relativeSpaceBone {boneName=pelvis}
    simSpaceSettings {maxAngularAcceleration=30, maxAngularVelocity=5, simSpaceAngularAlpha=0.7}
   }
   n12 = Anim Dynamics - Bone: dyn_hair_Front_C_01(ComponentPose=@n11, Alpha=1, ExternalForce="0,0,0") {
    angularDampingOverride=1
    angularSpringConstant=20000
    bAngularSpring=true
    bOverrideAngularDamping=true
    bUseGravityOverride=true
    gravityScale=0.35
    simulationSpace=BoneRelative
    alphaBoolBlend {blendOption=Linear}
    boundBone {boneName=dyn_hair_Front_C_01}
    componentAppliedLinearAccClamp=<3000 3000 3000>
    componentLinearAccScale=<1 1 1>
    componentLinearVelScale=<1 1 1>
    gravityOverride {z=-980}
    physicsBodyDefinitions [{
      boundBone {boneName=dyn_hair_Front_C_01}
      boxExtents=<20 30 30>
      constraintSetup {angularLimitsMax=<5 0 5>, angularLimitsMin=<-5 -5 -5>, angularTarget=<1 0 0>}
      localJointOffset=<4 0 0>
     }]
    relativeSpaceBone {boneName=pelvis}
    simSpaceSettings {maxAngularAcceleration=30, maxAngularVelocity=5, simSpaceAngularAlpha=0.7}
   }
   n13 = Anim Dynamics - Bone: dyn_hair_Front_C_02(ComponentPose=@n12, Alpha=1, ExternalForce="0,0,0") {
    angularDampingOverride=1
    angularSpringConstant=20000
    bAngularSpring=true
    bOverrideAngularDamping=true
    bUseGravityOverride=true
    gravityScale=0.35
    simulationSpace=BoneRelative
    alphaBoolBlend {blendOption=Linear}
    boundBone {boneName=dyn_hair_Front_C_02}
    componentAppliedLinearAccClamp=<3000 3000 3000>
    componentLinearAccScale=<1 1 1>
    componentLinearVelScale=<1 1 1>
    gravityOverride {z=-980}
    physicsBodyDefinitions [{
      boundBone {boneName=dyn_hair_Front_C_02}
      boxExtents=<20 30 30>
      constraintSetup {angularLimitsMax=<5 0 5>, angularLimitsMin=<-5 -5 -5>, angularTarget=<1 0 0>}
      localJointOffset=<4 0 0>
     }]
    relativeSpaceBone {boneName=pelvis}
    simSpaceSettings {maxAngularAcceleration=30, maxAngularVelocity=5, simSpaceAngularAlpha=0.7}
   }
   Save cached pose 'Cache PostDyn'(Pose=Component To Local(ComponentPose=@n13)) {cacheName=Cache PostDyn}
  ]}
 EventGraph {nodes [
   Event OnCharacterCustomizationLoaded()
   > n1 = Cast To PlayerPawn_Athena(Object=TryGetPawnOwner())
   > Set Body_Mesh(Body_Mesh=GetSkeletalMeshForPartType(self=@n1))
   n2 = CheckForBackbling()
   > Set isWearingBackbling(isWearingBackbling=AND Boolean!(A=@n2, B=NOT Boolean(A=@n2.bIsBackblingACape), C:bool=NOT Boolean(A=@n2.bIsBackblingLong)) -> (ReturnValue))
   > Set Is Wearing Cape(Is Wearing Cape=OR Boolean(A=@n2.bIsBackblingACape, B=@n2.bIsBackblingLong))
   K2Node_Event_1 = CustomEvent!() -> (OutputDelegate, then, DeltaTimeX:float=~) {is=title K2Node_CustomEvent}
   > n3 = Branch(Condition=Get bIsSkyDiving())
   > n4 = Branch(Condition=Get bIsParachuteOpened())
   > Set Override World Gravity(Override World Gravity:Vector="0,0,-300") -> (Output_Get:Vector)
   Set Override World Gravity(execute=@n4.else, Override World Gravity="0,0,500")
   Set Override World Gravity(execute=@n3.else, Override World Gravity="0,0,-980")
   Event BlueprintUpdateAnimation()
   > n5 = Sequence!(execute) -> (then_0, then_1, then_2:Exec)
   > Branch(Condition=Get bIsInFrontEnd())
   > Set Custom Idle is Playing Float(Custom Idle is Playing Float:double=GetCurveValue(self=GetOwnerFrontEndAnimInstance(), CurveName=Frontend_Idle_Bypass)) -> (Output_Get:double)
   n6 = Branch(execute=@n5.then_1, Condition=Get bIsParachuteOpened())
   > Set ispParachuteOpened(ispParachuteOpened:double=0) -> (Output_Get:double)
   Set ispParachuteOpened(execute=@n6.else, ispParachuteOpened=1)
   n7 = "Gravity Override State Function (Parameter Asset) (Nativized)"(
    execute=@n5.then_2
    CharacterPartAnimInstance=Self-Reference()
    GameTimeInSeconds=GetGameTimeInSeconds()
    DebugJointName=spine_05
    Parameters=/BRCosmetics/Characters/Player/Female/Medium/Bodies/F_MED_BaskIsle/Meshes/F_MED_BaskIsle_Parameters
   )
   > Set Gravity Override(Gravity Override=@n7)
   > Set RBAN Settings Out(RBAN Settings Out=@n7.OutRigidBodyAnimNodeParameters)
   Get bIsSkydiveDiveMode()
  ]}
]
variables [
 Body_Mesh {type=SkeletalMeshComponent}
 Custom Idle is Playing Float {type=number}
 Gravity Override {type=Vector}
 Is Wearing Cape {type=boolean}
 ispParachuteOpened {type=number}
 isWearingBackbling {type=boolean}
 Override World Gravity {type=Vector}
 RBAN Settings Out {type=FortRigidBodyAnimNodeParameters}
]
